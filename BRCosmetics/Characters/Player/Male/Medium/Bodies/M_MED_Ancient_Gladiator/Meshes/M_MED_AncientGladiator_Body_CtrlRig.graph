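# Control Rig for the Ancient Gladiator body cosmetic, procedurally correcting loincloth and left pauldron bones.
# The executed chain aims a helper bone between both calves, constrains front and back loincloth to it, then adds crouch and left-thigh offsets.
# Separate chains, unlinked from Begin Execution, hold pauldron and further loincloth corrections.
# These remap bone angles and calf distance into additive weights, with some pauldron weights using targeting, skydiving and hang-gliding variables.
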ControlRigBlueprint /BRCosmetics/Characters/Player/Male/Medium/Bodies/M_MED_Ancient_Gladiator/Meshes/M_MED_AncientGladiator_Body_CtrlRig
graphs [
 RigVMFunctionLibrary {links [], nodes []}
 RigVMModel {nodes [
   n1 = &RigVMFunction_MathTransformMakeRelative(
    Global=(Rotation=(X=0,Y=0,Z=0,W=1),Translation=(X=0,Y=0,Z=0),Scale3D=(X=1,Y=1,Z=1))
    Parent=(Rotation=(X=0,Y=0,Z=0,W=1),Translation=(X=0,Y=0,Z=0),Scale3D=(X=1,Y=1,Z=1))
    Global.Rotation=&RigUnit_GetBoneTransform(Bone=clavicle_l).Transform.Rotation
    Parent.Rotation=/Script/ControlRig.RigUnit_GetInitialBoneTransform(Bone=clavicle_l, Space=GlobalSpace) {is=struct RigVMUnitNode}.Transform.Rotation
   )
   n2 = &RigVMFunction_MathQuaternionToEuler(RotationOrder=ZYX, Value=@n1.Local.Rotation)
   /Script/ControlRig.RigUnit_ModifyBoneTransforms(
    BoneToModify=((Bone="dyn_defaultArmorPauldron",Transform=(Rotation=(X=0,Y=0,Z=0,W=1),Translation=(X=-2,Y=-1.5,Z=1.4),Scale3D=(X=1,Y=1.2,Z=1))))
    ExecutePin=()
    Mode=AdditiveLocal
    WeightMaximum=1
    WeightMinimum=0
    Weight=&RigVMFunction_MathFloatMul(
     A=1
     B=&RigVMFunction_MathFloatRemap(SourceMaximum=67, SourceMinimum=78, TargetMinimum=0, bClamp=true, Value=@n2.Result.Z).Result
    ) {template="Multiply::Execute(in A,in B,out Result)"}.Result
    BoneToModify.0.Transform.Rotation=&RigVMFunction_MathQuaternionFromEuler(Euler=(X=0,Y=0,Z=0)).Result
   ) {is=struct RigVMUnitNode}
   n3 = &RigVMFunction_MathQuaternionToEuler(RotationOrder=ZYX, Value=&RigUnit_GetRelativeBoneTransform(Bone=upperarm_l, Space=spine_05).Transform.Rotation)
   n4 = /Script/RigVM.RigVMFunction_MathFloatAbs(Value=@n3.Result.Z) {is=struct RigVMUnitNode, template="Absolute::Execute(in Value,out Result)"}
   n5 = &RigVMFunction_MathFloatRemap(SourceMaximum=68, SourceMinimum=140, TargetMinimum=0, bClamp=true, Value=@n4.Result)
   n6 = RigVMVariableNode() {variable=targeting}
   > &RigUnit_ModifyBoneTransforms(
    BoneToModify=((Bone="dyn_defaultArmorPauldron",Transform=(Rotation=(X=0,Y=0,Z=0,W=1),Translation=(X=0,Y=-1,Z=0),Scale3D=(X=1,Y=1.1,Z=1))))
    Mode=AdditiveLocal
    WeightMaximum=1
    WeightMinimum=0
    Weight=&RigVMFunction_MathFloatMul(A=@n5.Result, B=@n6.Value) {template="Multiply::Execute(in A,in B,out Result)"}.Result
    BoneToModify.0.Transform.Rotation=&RigVMFunction_MathQuaternionFromEuler(Euler=(X=-5,Y=0,Z=0)).Result
   )
   n7 = &RigUnit_GetBoneTransform(Bone=calf_r)
   n8 = &RigUnit_GetBoneTransform(Bone=calf_l)
   > &RigUnit_ModifyBoneTransforms(
    BoneToModify=((Bone="dyn_loinClothFront_01_c_Base",Transform=(Rotation=(X=0,Y=0,Z=0,W=1),Translation=(X=0,Y=0,Z=0),Scale3D=(X=1,Y=1,Z=1))))
    Mode=AdditiveLocal
    WeightMaximum=1
    WeightMinimum=0
    Weight=&RigVMFunction_MathFloatRemap(
     SourceMaximum=19
     SourceMinimum=50
     Value=&RigVMFunction_MathVectorDistance(B=@n7.Transform.Translation, A=@n8.Transform.Translation).Result
    ).Result
    BoneToModify.0.Transform.Rotation=&RigVMFunction_MathQuaternionFromEuler(Euler=(X=0,Y=0,Z=-10)).Result
   )
   n9 = &RigUnit_GetRelativeBoneTransform(Bone=thigh_r, Space=pelvis)
   n10 = &RigVMFunction_MathFloatRemap(
    SourceMaximum=72
    SourceMinimum=150
    TargetMinimum=0
    bClamp=true
    Value=&RigVMFunction_MathFloatAbs(Value=&RigVMFunction_MathQuaternionToEuler(RotationOrder=ZYX, Value=@n9.Transform.Rotation).Result.Z) {template="Absolute::Execute(in Value,out Result)"}.Result
   )
   &RigUnit_ModifyBoneTransforms(
    BoneToModify=((Bone="dyn_loinClothFront_01_r",Transform=(Rotation=(X=0,Y=0,Z=0,W=1),Translation=(X=0,Y=-11,Z=-4),Scale3D=(X=1,Y=1,Z=1))))
    ExecutePin=()
    Mode=AdditiveLocal
    WeightMaximum=1
    WeightMinimum=0
    Weight=@n10.Result
    BoneToModify.0.Transform.Rotation=&RigVMFunction_MathQuaternionFromEuler(Euler=(X=0,Y=0,Z=0)).Result
   )
   > &RigUnit_ModifyBoneTransforms(
    BoneToModify=((Bone="dyn_loinClothFront_01_c_Base",Transform=(Rotation=(X=0,Y=0,Z=0,W=1),Translation=(X=0,Y=0,Z=3),Scale3D=(X=1,Y=1,Z=1))))
    Mode=AdditiveLocal
    WeightMaximum=1
    Weight=@n10.Result
    BoneToModify.0.Transform.Rotation=&RigVMFunction_MathQuaternionFromEuler(Euler=(X=-25,Y=0,Z=0)).Result
   )
   > &RigUnit_ModifyBoneTransforms(
    BoneToModify=((Bone="dyn_lightArmorPauldron",Transform=(Rotation=(X=0,Y=0,Z=0,W=1),Translation=(X=0,Y=0,Z=0.5),Scale3D=(X=1,Y=1,Z=1))))
    Mode=AdditiveLocal
    Weight=1
    WeightMaximum=1
    WeightMinimum=0
    BoneToModify.0.Transform.Rotation=&RigVMFunction_MathQuaternionFromEuler(Euler=(X=0,Y=-5,Z=0)).Result
   )
   > &RigUnit_ModifyBoneTransforms(
    BoneToModify=((Bone="dyn_lightArmorPauldron",Transform=(Rotation=(X=0,Y=0,Z=0,W=1),Translation=(X=0,Y=-1,Z=0),Scale3D=(X=1,Y=1,Z=1))))
    Mode=AdditiveLocal
    WeightMaximum=1
    WeightMinimum=0
    Weight=@n5.Result
    BoneToModify.0.Transform.Rotation=&RigVMFunction_MathQuaternionFromEuler(Euler=(X=50,Y=-5,Z=0)).Result
   )
   n11 = &RigVMFunction_MathFloatAbs(Value=@n3.Result.X) {template="Absolute::Execute(in Value,out Result)"}
   > &RigUnit_ModifyBoneTransforms(
    BoneToModify=((Bone="dyn_lightArmorPauldron",Transform=(Rotation=(X=0,Y=0,Z=0,W=1),Translation=(X=0,Y=0,Z=0),Scale3D=(X=1,Y=1,Z=1))))
    Mode=AdditiveLocal
    WeightMaximum=1
    WeightMinimum=0
    Weight=&RigVMFunction_MathFloatMul(
     A=/Script/RigVM.RigVMFunction_AccumulateFloatLerp(Blend=10, bIntegrateDeltaTime=true, InitialValue=@n6.Value, TargetValue=@n6.Value) {is=struct RigVMUnitNode, template="AccumulateLerp::Execute(in TargetValue,in InitialValue,in Blend,in bIntegrateDeltaTime,out Result)"}.Result
     B=&RigVMFunction_MathFloatRemap(SourceMaximum=109, SourceMinimum=144, bClamp=true, Value=@n11.Result).Result
    ) {template="Multiply::Execute(in A,in B,out Result)"}.Result
    BoneToModify.0.Transform.Rotation=&RigVMFunction_MathQuaternionFromEuler(Euler=(X=-20,Y=0,Z=0)).Result
   )
   > &RigUnit_ModifyBoneTransforms(
    BoneToModify=((Bone="dyn_lightArmorPauldron",Transform=(Rotation=(X=0,Y=0,Z=0,W=1),Translation=(X=0,Y=0,Z=0),Scale3D=(X=1,Y=1,Z=1))))
    Mode=AdditiveLocal
    WeightMaximum=1
    WeightMinimum=0
    Weight=&RigVMFunction_MathFloatRemap(
     SourceMaximum=120
     SourceMinimum=140
     bClamp=true
     Value=&RigVMFunction_MathFloatAbs(Value=@n3.Result.Z) {template="Absolute::Execute(in Value,out Result)"}.Result
    ).Result
    BoneToModify.0.Transform.Rotation=&RigVMFunction_MathQuaternionFromEuler(Euler=(X=-30,Y=0,Z=0)).Result
   )
   n12 = &RigVMFunction_MathQuaternionToEuler(RotationOrder=ZYX, Value=&RigUnit_GetRelativeBoneTransform(Bone=spine_05, Space=spine_01).Transform.Rotation)
   > &RigUnit_ModifyBoneTransforms(
    BoneToModify=((Bone="dyn_lightArmorPauldron",Transform=(Rotation=(X=0,Y=0,Z=0,W=1),Translation=(X=0,Y=-1,Z=2),Scale3D=(X=1,Y=1,Z=1))))
    Mode=AdditiveLocal
    WeightMaximum=1
    WeightMinimum=0
    Weight=&RigVMFunction_MathFloatMul(
     A=@n5.Result
     B=&RigVMFunction_MathFloatRemap(SourceMaximum=-20, SourceMinimum=-5, TargetMinimum=0, bClamp=true, Value=@n12.Result.Y).Result
    ) {template="Multiply::Execute(in A,in B,out Result)"}.Result
    BoneToModify.0.Transform.Rotation=&RigVMFunction_MathQuaternionFromEuler(Euler=(X=0,Y=0,Z=0)).Result
   )
   > &RigUnit_ModifyBoneTransforms(
    BoneToModify=((Bone="dyn_lightArmorPauldron",Transform=(Rotation=(X=0,Y=0,Z=0,W=1),Translation=(X=0,Y=0,Z=0.85),Scale3D=(X=1,Y=1,Z=1))))
    Mode=AdditiveLocal
    WeightMaximum=1
    WeightMinimum=0
    Weight=&RigVMFunction_MathFloatMul(
     A=&RigVMFunction_MathFloatRemap(SourceMaximum=175, SourceMinimum=146, TargetMinimum=0, bClamp=true, Value=@n4.Result).Result
     B=@n6.Value
    ) {template="Multiply::Execute(in A,in B,out Result)"}.Result
    BoneToModify.0.Transform.Rotation=&RigVMFunction_MathQuaternionFromEuler(Euler=(X=-12,Y=0,Z=0)).Result
   )
   n13 = RigVMVariableNode() {variable=skydiving}
   > &RigUnit_ModifyBoneTransforms(
    BoneToModify=((Bone="dyn_lightArmorPauldron",Transform=(Rotation=(X=0,Y=0,Z=0,W=1),Translation=(X=2.5,Y=0,Z=1),Scale3D=(X=1,Y=1,Z=1))))
    Mode=AdditiveLocal
    WeightMaximum=1
    WeightMinimum=0
    BoneToModify.0.Transform.Rotation=&RigVMFunction_MathQuaternionFromEuler(Euler=(X=0,Y=20,Z=0)).Result
    Weight=@n13.Value
   )
   > &RigUnit_ModifyBoneTransforms(
    BoneToModify=((Bone="dyn_lightArmorPauldron",Transform=(Rotation=(X=0,Y=0,Z=0,W=1),Translation=(X=2,Y=0,Z=2.25),Scale3D=(X=1,Y=1,Z=1))))
    Mode=AdditiveLocal
    WeightMaximum=1
    WeightMinimum=0
    Weight=&RigVMFunction_MathFloatMul(A=@n13.Value, B=RigVMVariableNode() {variable=hanggliding}.Value) {template="Multiply::Execute(in A,in B,out Result)"}.Result
    BoneToModify.0.Transform.Rotation=&RigVMFunction_MathQuaternionFromEuler(Euler=(X=15,Y=0,Z=5)).Result
   )
   n14 = &RigVMFunction_MathFloatRemap(SourceMaximum=137, SourceMinimum=155, bClamp=true, Value=@n11.Result)
   n15 = /Script/RigVM.RigVMFunction_MathFloatSub(A=1, B=@n5.Result) {is=struct RigVMUnitNode, template="Subtract::Execute(in A,in B,out Result)"}
   > &RigUnit_ModifyBoneTransforms(
    BoneToModify=((Bone="dyn_lightArmorPauldron",Transform=(Rotation=(X=0,Y=0,Z=0,W=1),Translation=(X=0,Y=0,Z=0),Scale3D=(X=1,Y=1,Z=1))))
    Mode=AdditiveLocal
    WeightMaximum=1
    WeightMinimum=0
    Weight=&RigVMFunction_MathFloatMul(
     B=&RigVMFunction_MathFloatMul(B=@n14.Result, A=@n15.Result) {template="Multiply::Execute(in A,in B,out Result)"}.Result
     A=&RigVMFunction_MathFloatSub(A=1, B=@n6.Value) {template="Subtract::Execute(in A,in B,out Result)"}.Result
    ) {template="Multiply::Execute(in A,in B,out Result)"}.Result
    BoneToModify.0.Transform.Rotation=&RigVMFunction_MathQuaternionFromEuler(Euler=(X=-40,Y=0,Z=0)).Result
   )
   RigUnit_BeginExecution = &RigUnit_BeginExecution()
   /Script/ControlRig.RigUnit_TransformConstraint(
    BaseTransform=(Rotation=(X=0,Y=0,Z=0,W=1),Translation=(X=0,Y=0,Z=0),Scale3D=(X=1,Y=1,Z=1))
    BaseTransformSpace=GlobalSpace
    Bone=dyn_kneeAim_l
    Targets="((Transform=(Rotation=(X=0,Y=0,Z=0,W=1),Translation=(X=0,Y=0,Z=0),Scale3D=(X=1,Y=1,Z=1)),Weight=1,bMaintainOffset=True,Filter=(TranslationFilter=(bX=True,bY=True,bZ=True),RotationFilter=(bX=True,bY=True,bZ=True),ScaleFilter=(bX=True,bY=True,bZ=True))),(Tra ...(truncated)"
    bUseInitialTransforms=True
    Targets.0.Transform.Translation=&RigUnit_GetBoneTransform(Bone=calf_l).Transform.Translation
    Targets.1.Transform.Translation=&RigUnit_GetBoneTransform(Bone=calf_r).Transform.Translation
    ExecutePin=@RigUnit_BeginExecution.ExecutePin
   ) {is=struct RigVMUnitNode}
   RigVMStructNode_0 = &RigUnit_GetBoneTransform(Bone=calf_r)
   > /Script/ControlRig.RigUnit_AimBone(
    Bone=dyn_kneeAim_l
    DebugSettings=(bEnabled=true,Scale=10,WorldOffset=(Rotation=(X=0,Y=0,Z=0,W=1),Translation=(X=0,Y=0,Z=0),Scale3D=(X=1,Y=1,Z=1)))
    Primary=(Weight=1,Axis=(X=0,Y=-0.1,Z=1),Target=(X=1,Y=0,Z=0),Kind=Location)
    Secondary=(Weight=1,Axis=(X=1,Y=-0.11,Z=-0.085),Target=(X=1,Y=0,Z=0),Kind=Location)
    Weight=1
    bPropagateToChildren=True
    Secondary.Target=&RigUnit_GetBoneTransform(Bone=pelvis).Transform.Translation
    Primary.Target=@RigVMStructNode_0.Transform.Translation
   ) {is=struct RigVMUnitNode}
   n16 = &RigUnit_GetBoneTransform(Bone=dyn_loinClothBack_01_Base)
   > &RigUnit_TransformConstraint(
    BaseBone=pelvis
    BaseTransform=(Rotation=(X=0,Y=0,Z=0,W=1),Translation=(X=0,Y=0,Z=0),Scale3D=(X=1,Y=1,Z=1))
    BaseTransformSpace=GlobalSpace
    Bone=dyn_loinClothBack_01_Base
    Targets="((Transform=(Rotation=(X=0,Y=0,Z=0,W=1),Translation=(X=0,Y=0,Z=0),Scale3D=(X=1,Y=1,Z=1)),Weight=0.2,bMaintainOffset=True,Filter=(TranslationFilter=(bX=True,bY=True,bZ=True),RotationFilter=(bX=True,bY=True,bZ=True),ScaleFilter=(bX=True,bY=True,bZ=True))),(T ...(truncated)"
    bUseInitialTransforms=True
    Targets.0.Transform.Rotation=&RigUnit_GetBoneTransform(Bone=dyn_kneeAim_l).Transform.Rotation
    Targets.1.Transform=@n16.Transform
    Targets.0.Transform.Translation=@n16.Transform.Translation
   )
   n17 = &RigUnit_GetBoneTransform(Bone=dyn_loinClothBack_01_Base)
   > &RigUnit_TransformConstraint(
    BaseBone=pelvis
    BaseTransform=(Rotation=(X=0,Y=0,Z=0,W=1),Translation=(X=0,Y=0,Z=0),Scale3D=(X=1,Y=1,Z=1))
    BaseTransformSpace=GlobalSpace
    Bone=dyn_loinClothBack_01_Base
    Targets="((Transform=(Rotation=(X=0,Y=0,Z=0,W=1),Translation=(X=0,Y=0,Z=0),Scale3D=(X=1,Y=1,Z=1)),Weight=0.15,bMaintainOffset=True,Filter=(TranslationFilter=(bX=True,bY=True,bZ=True),RotationFilter=(bX=True,bY=True,bZ=True),ScaleFilter=(bX=True,bY=True,bZ=True))),( ...(truncated)"
    bUseInitialTransforms=True
    Targets.0.Transform.Translation=&RigUnit_GetBoneTransform(Bone=thigh_twist_01_l).Transform.Translation
    Targets.1.Transform.Translation=&RigUnit_GetBoneTransform(Bone=thigh_twist_01_r).Transform.Translation
    Targets.2.Transform=@n17.Transform
    Targets.0.Transform.Rotation=@n17.Transform.Rotation
    Targets.1.Transform.Rotation=@n17.Transform.Rotation
   )
   n18 = &RigUnit_GetBoneTransform(Bone=dyn_loinClothFront_01_c_Base)
   > &RigUnit_TransformConstraint(
    BaseTransform=(Rotation=(X=0,Y=0,Z=0,W=1),Translation=(X=0,Y=0,Z=0),Scale3D=(X=1,Y=1,Z=1))
    BaseTransformSpace=GlobalSpace
    Bone=dyn_loinClothFront_01_c_Base
    Targets="((Transform=(Rotation=(X=0,Y=0,Z=0,W=1),Translation=(X=0,Y=0,Z=0),Scale3D=(X=1,Y=1,Z=1)),Weight=1,bMaintainOffset=True,Filter=(TranslationFilter=(bX=True,bY=True,bZ=True),RotationFilter=(bX=True,bY=True,bZ=True),ScaleFilter=(bX=True,bY=True,bZ=True))),(Tra ...(truncated)"
    bUseInitialTransforms=True
    Targets.0.Transform.Rotation=&RigUnit_GetBoneTransform(Bone=dyn_kneeAim_l).Transform.Rotation
    Targets.1.Transform=@n18.Transform
    Targets.0.Transform.Translation=@n18.Transform.Translation
   )
   n19 = RigVMVariableNode() {variable=crouching}
   > /Script/ControlRig.RigUnit_AddBoneTransform(
    Bone=dyn_loinClothFront_01_c_Base
    Transform=(Rotation=(X=0,Y=0,Z=0,W=1),Translation=(X=0,Y=-5,Z=0),Scale3D=(X=1,Y=1,Z=1))
    bPropagateToChildren=True
    Transform.Rotation=&RigVMFunction_MathQuaternionFromEuler(Euler=(X=0,Y=0,Z=-10)).Result
    Weight=@n19.Value
   ) {is=struct RigVMUnitNode}
   n20 = &RigUnit_GetRelativeBoneTransform(Bone=thigh_l, Space=pelvis)
   n21 = &RigVMFunction_MathFloatRemap(
    SourceMaximum=105
    SourceMinimum=20
    TargetMinimum=0
    bClamp=true
    Value=&RigVMFunction_MathFloatAbs(Value=&RigVMFunction_MathQuaternionToEuler(RotationOrder=ZYX, Value=@n20.Transform.Rotation).Result.Z) {template="Absolute::Execute(in Value,out Result)"}.Result
   )
   > &RigUnit_ModifyBoneTransforms(
    BoneToModify=((Bone="dyn_loinClothFront_01_l",Transform=(Rotation=(X=0,Y=0,Z=0,W=1),Translation=(X=0,Y=-5,Z=0),Scale3D=(X=1,Y=1,Z=1))))
    Mode=AdditiveLocal
    WeightMaximum=1
    Weight=@n21.Result
    BoneToModify.0.Transform.Rotation=&RigVMFunction_MathQuaternionFromEuler(Euler=(X=-20,Y=0,Z=0)).Result
   )
   > &RigUnit_ModifyBoneTransforms(
    BoneToModify=((Bone="dyn_loinClothFront_01_c_Base",Transform=(Rotation=(X=0,Y=0,Z=0,W=1),Translation=(X=0,Y=0,Z=-3),Scale3D=(X=1,Y=1,Z=1))))
    Mode=AdditiveLocal
    WeightMaximum=1
    Weight=@n21.Result
    BoneToModify.0.Transform.Rotation=&RigVMFunction_MathQuaternionFromEuler(Euler=(X=25,Y=0,Z=0)).Result
   )
   &RigVMFunction_MathFloatMul(A=@n10.Result, B=@n19.Value) {template="Multiply::Execute(in A,in B,out Result)"}
   &RigVMFunction_MathFloatMul(A=1, B=1) {template="Multiply::Execute(in A,in B,out Result)"}
   &RigVMFunction_MathFloatRemap(SourceMaximum=67, SourceMinimum=78, TargetMinimum=0, bClamp=true, Value=@n2.Result.X)
   &RigVMFunction_MathFloatRemap(SourceMaximum=1, SourceMinimum=0, TargetMinimum=0.5, bClamp=true, Value=@n19.Value)
   &RigVMFunction_MathFloatRemap(SourceMaximum=1, SourceMinimum=0, TargetMaximum=0, TargetMinimum=0.5, bClamp=true, Value=@n19.Value)
   /Script/RigVM.RigVMFunction_MathVectorAngle(B=(X=0,Y=0,Z=0), A=@n8.Transform.Translation) {is=struct RigVMUnitNode, template="AngleBetween::Execute(in A,in B,out Result)"}
  ]}
]
hierarchy [
 root {loc=<0 0 0>, rot=<0 0 0 1>}
 attach {parent=root, loc=<0 0 0>, rot=<0 0 0 1>}
 pelvis {parent=root, loc=<0 -0.7071 93.1191>, rot=<0.0127 0.707 -0.0127 -0.707>}
 spine_01 {parent=pelvis, loc=<6.7494 -0.6388 0>, rot=<0 0 0.034 -0.9994>}
 spine_02 {parent=spine_01, loc=<7.2518 -0.035 0>, rot=<0 0 0.0238 -0.9997>}
 spine_03 {parent=spine_02, loc=<6.0104 0 0>, rot=<0 0 0.0149 -0.9999>}
 spine_04 {parent=spine_03, loc=<8.1024 0 0>, rot=<0 0 -0.0132 -0.9999>}
 spine_05 {parent=spine_04, loc=<10.0146 0 0>, rot=<0 0 -0.0015 -1>}
 clavicle_l {parent=spine_05, loc=<4.3575 -0.1466 -2.3223>, rot=<0.0465 -0.7289 -0.009 -0.6829>}
 upperarm_l {parent=clavicle_l, loc=<15.5111 0 0>, rot=<0.0226 -0.3234 -0.0088 -0.9459>}
 lowerarm_l {parent=upperarm_l, loc=<26.6514 0 0>, rot=<0.0391 -0.0099 -0.2453 -0.9686>}
 hand_l {parent=lowerarm_l, loc=<22.3491 0 0>, rot=<-0.7081 -0.0507 0.0975 -0.6975>}
 index_metacarpal_l {parent=hand_l, loc=<1.6535 0.0183 -2.96>, rot=<0.0158 -0.0124 -0.0173 -0.9996>}
 index_01_l {parent=index_metacarpal_l, loc=<8.3008 0.0357 -0.0524>, rot=<-0.0044 0.0485 0.0924 -0.9945>}
 index_02_l {parent=index_01_l, loc=<3.3587 -0.3004 0.0452>, rot=<-0.002 0.0027 0.3225 -0.9466>}
 index_03_l {parent=index_02_l, loc=<2.5578 -0.123 -0.0252>, rot=<-0.002 0.0067 0.2558 -0.9667>}
 middle_metacarpal_l {parent=hand_l, loc=<2.1856 -0.1725 -0.5984>, rot=<0.006 0.0168 -0.0329 -0.9993>}
 middle_01_l {parent=middle_metacarpal_l, loc=<7.3275 -0.2447 0.0129>, rot=<-0.0208 0.0042 0.0858 -0.9961>}
 middle_02_l {parent=middle_01_l, loc=<4.0123 0.1401 0.1175>, rot=<-0.0218 0.0258 0.3512 -0.9357>}
 middle_03_l {parent=middle_02_l, loc=<2.6077 0.0534 0.1373>, rot=<0.0082 0.0188 0.1708 -0.9851>}
 pinky_metacarpal_l {parent=hand_l, loc=<1.7616 -0.9753 3.3814>, rot=<-0.0397 0.0814 -0.069 -0.9935>}
 pinky_01_l {parent=pinky_metacarpal_l, loc=<7.1709 -0.4831 -0.1094>, rot=<-0.0654 -0.0035 0.2242 -0.9723>}
 pinky_02_l {parent=pinky_01_l, loc=<2.8274 0.0893 -0.1043>, rot=<0.0037 0.0011 0.2854 -0.9584>}
 pinky_03_l {parent=pinky_02_l, loc=<2.2046 0.0307 -0.0078>, rot=<-0.0002 0.0098 0.1445 -0.9895>}
 ring_metacarpal_l {parent=hand_l, loc=<2.2074 -0.3625 1.3638>, rot=<-0.0284 0.0722 -0.0458 -0.9959>}
 ring_01_l {parent=ring_metacarpal_l, loc=<7.0295 -0.6197 -0.1446>, rot=<-0.0489 -0.0252 0.1892 -0.9804>}
 ring_02_l {parent=ring_01_l, loc=<3.6174 0.4646 -0.1027>, rot=<0 0 0.2806 -0.9598>}
 ring_03_l {parent=ring_02_l, loc=<2.4152 0.1141 0.0199>, rot=<0 0 0.1189 -0.9929>}
 thumb_01_l {parent=hand_l, loc=<1.3118 -0.3191 -2.459>, rot=<0.6546 -0.4042 0.0664 -0.6354>}
 thumb_02_l {parent=thumb_01_l, loc=<5.9336 -0.0448 0.0401>, rot=<-0.0037 0.0213 0.298 -0.9543>}
 thumb_03_l {parent=thumb_02_l, loc=<2.2223 0.3682 -0.0907>, rot=<0.0118 -0.013 0.0066 -0.9998>}
 weapon_l {parent=hand_l, loc=<-0.1974 -3.4111 0.3811>, rot=<0 0 0 -1>}
 lowerarm_twist_01_l {parent=lowerarm_l, loc=<17.8549 0 0>, rot=<0 0 0 1>}
 lowerarm_twist_02_l {parent=lowerarm_l, loc=<9.6199 0 0>, rot=<0 0 0 1>}
 upperarm_twist_01_l {parent=upperarm_l, loc=<0.0041 0 0>, rot=<0 0 0 -1>}
 dyn_lightArmorPauldron {parent=upperarm_twist_01_l, loc=<0 0 0>, rot=<0 0 0 -1>}
 dyn_defaultArmorPauldron {parent=upperarm_twist_01_l, loc=<0 0 0>, rot=<0 0 0 -1>}
 upperarm_twist_02_l {parent=upperarm_l, loc=<13.7941 0 0>, rot=<0 0 0 -1>}
 clavicle_r {parent=spine_05, loc=<4.3575 -0.1466 2.3223>, rot=<0.7289 0.0465 -0.6829 0.009>}
 upperarm_r {parent=clavicle_r, loc=<-15.5111 0 0>, rot=<0.0226 -0.3234 -0.0088 -0.9459>}
 lowerarm_r {parent=upperarm_r, loc=<-26.6514 0 0>, rot=<0.0391 -0.0099 -0.2453 -0.9686>}
 hand_r {parent=lowerarm_r, loc=<-22.3492 0 0>, rot=<-0.7081 -0.0507 0.0975 -0.6975>}
 index_metacarpal_r {parent=hand_r, loc=<-1.6534 -0.0181 2.9601>, rot=<0.0158 -0.0124 -0.0173 -0.9996>}
 index_01_r {parent=index_metacarpal_r, loc=<-8.3008 -0.0357 0.0524>, rot=<-0.0044 0.0485 0.0924 -0.9945>}
 index_02_r {parent=index_01_r, loc=<-3.3589 0.3003 -0.0453>, rot=<-0.002 0.0027 0.3225 -0.9466>}
 index_03_r {parent=index_02_r, loc=<-2.5578 0.1231 0.0251>, rot=<-0.002 0.0067 0.2558 -0.9667>}
 middle_metacarpal_r {parent=hand_r, loc=<-2.1856 0.1726 0.5984>, rot=<0.006 0.0168 -0.0329 -0.9993>}
 middle_01_r {parent=middle_metacarpal_r, loc=<-7.3273 0.2447 -0.0129>, rot=<-0.0208 0.0042 0.0858 -0.9961>}
 middle_02_r {parent=middle_01_r, loc=<-4.0142 -0.1411 -0.07>, rot=<-0.0217 0.0259 0.3512 -0.9357>}
 middle_03_r {parent=middle_02_r, loc=<-2.6074 -0.0534 -0.1312>, rot=<0.0083 0.0188 0.1708 -0.9851>}
 pinky_metacarpal_r {parent=hand_r, loc=<-1.7617 0.9753 -3.3814>, rot=<-0.0397 0.0814 -0.069 -0.9935>}
 pinky_01_r {parent=pinky_metacarpal_r, loc=<-7.1707 0.4832 0.1094>, rot=<-0.0654 -0.0035 0.2242 -0.9723>}
 pinky_02_r {parent=pinky_01_r, loc=<-2.8274 -0.0892 0.1043>, rot=<0.0037 0.0011 0.2854 -0.9584>}
 pinky_03_r {parent=pinky_02_r, loc=<-2.2045 -0.031 0.0078>, rot=<-0.0002 0.0098 0.1445 -0.9895>}
 ring_metacarpal_r {parent=hand_r, loc=<-2.2073 0.3627 -1.3638>, rot=<-0.0284 0.0722 -0.0458 -0.9959>}
 ring_01_r {parent=ring_metacarpal_r, loc=<-7.0297 0.6197 0.1446>, rot=<-0.0489 -0.0252 0.1892 -0.9804>}
 ring_02_r {parent=ring_01_r, loc=<-3.6172 -0.4648 0.1027>, rot=<0 0 0.2806 -0.9598>}
 ring_03_r {parent=ring_02_r, loc=<-2.4152 -0.1141 -0.0199>, rot=<0 0 0.1189 -0.9929>}
 thumb_01_r {parent=hand_r, loc=<-1.3119 0.3192 2.459>, rot=<0.6546 -0.4042 0.0664 -0.6354>}
 thumb_02_r {parent=thumb_01_r, loc=<-5.9335 0.0447 -0.0401>, rot=<-0.0037 0.0213 0.298 -0.9543>}
 thumb_03_r {parent=thumb_02_r, loc=<-2.2223 -0.3682 0.0907>, rot=<0.0118 -0.013 0.0066 -0.9998>}
 weapon_r {parent=hand_r, loc=<0.1975 3.4111 -0.3811>, rot=<0 0 0 -1>}
 lowerarm_twist_01_r {parent=lowerarm_r, loc=<-17.855 0 0>, rot=<0 0 0 1>}
 lowerarm_twist_02_r {parent=lowerarm_r, loc=<-9.6199 0 0>, rot=<0 0 0 1>}
 upperarm_twist_01_r {parent=upperarm_r, loc=<-0.0041 0 0>, rot=<0 0 0 1>}
 upperarm_twist_02_r {parent=upperarm_r, loc=<-13.7942 0 0>, rot=<0 0 0 1>}
 neck_01 {parent=spine_05, loc=<11.3138 0.1529 0>, rot=<0 0 -0.2003 -0.9797>}
 neck_02 {parent=neck_01, loc=<4.5627 0 0>, rot=<0 0 0.0504 -0.9987>}
 head {parent=neck_02, loc=<5.3914 0 0>, rot=<0 0 0.115 -0.9934>}
 thigh_l {parent=pelvis, loc=<-2.5693 -0.0109 -7.7501>, rot=<0.0012 0.0572 0.0155 -0.9982>}
 calf_l {parent=thigh_l, loc=<-39.0484 0 0>, rot=<0.0006 -0.0161 -0.0391 -0.9991>}
 calf_twist_01_l {parent=calf_l, loc=<-35.8131 -0.0967 0.0397>, rot=<0.0237 0.0095 -0.0035 -0.9997>}
 calf_twist_02_l {parent=calf_l, loc=<-18.8064 -0.0967 0.0397>, rot=<0.0237 0.0095 -0.0035 -0.9997>}
 foot_l {parent=calf_l, loc=<-42.561 0.1077 0.4292>, rot=<0.0418 -0.0455 0.0343 -0.9975>}
 ball_l {parent=foot_l, loc=<-6.1599 13.8264 0.6806>, rot=<0.0001 -0.0001 -0.7063 -0.7079>}
 dyn_kneeAim_l {parent=calf_l, loc=<0 0 0>, rot=<0 0 0 1>}
 thigh_twist_01_l {parent=thigh_l, loc=<-21.6303 0 0>, rot=<0.0013 0 0 -1>}
 dyn_kneeAimCollide_l {parent=thigh_l, loc=<-21.6303 0 0>, rot=<0.0013 0 0 -1>}
 thigh_r {parent=pelvis, loc=<-2.5691 -0.0109 7.7501>, rot=<0.0572 -0.0012 0.9982 0.0155>}
 calf_r {parent=thigh_r, loc=<39.0487 0 0>, rot=<0.0006 -0.0161 -0.0391 -0.9991>}
 calf_twist_01_r {parent=calf_r, loc=<35.8132 0.0967 -0.0397>, rot=<0.0237 0.0095 -0.0035 -0.9997>}
 calf_twist_02_r {parent=calf_r, loc=<18.8064 0.0967 -0.0397>, rot=<0.0237 0.0095 -0.0035 -0.9997>}
 foot_r {parent=calf_r, loc=<42.5611 -0.1077 -0.4292>, rot=<0.0418 -0.0455 0.0343 -0.9975>}
 ball_r {parent=foot_r, loc=<6.1598 -13.8264 -0.6806>, rot=<0.0001 -0.0001 -0.7063 -0.7079>}
 thigh_twist_01_r {parent=thigh_r, loc=<21.6303 0 0>, rot=<0.0013 0 0 -1>}
 dyn_backStrapOut_01_Base_l {parent=pelvis, loc=<4.7683 -12.0801 -7.0565>, rot=<0.1252 -0.9918 0.0039 -0.0241>}
 dyn_backStrapOut_01_l {parent=dyn_backStrapOut_01_Base_l, loc=<0 0 0>, rot=<0 0 0 -1>}
 dyn_backStrapOut_02_l {parent=dyn_backStrapOut_01_l, loc=<5.4806 0 0>, rot=<-0.0019 0.0089 -0.0727 -0.9973>}
 dyn_backStrapIn_01_Base_l {parent=pelvis, loc=<4.395 -13.3156 -2.5033>, rot=<0.124 -0.9921 0.003 -0.0183>}
 dyn_backStrapIn_01_l {parent=dyn_backStrapIn_01_Base_l, loc=<0 0 0>, rot=<0 0 0 1>}
 dyn_backStrapIn_02_l {parent=dyn_backStrapIn_01_l, loc=<5.4161 0 0>, rot=<0.0035 -0.0157 -0.063 -0.9979>}
 dyn_backStrapIn_01_Base_r {parent=pelvis, loc=<4.395 -13.3156 2.5033>, rot=<0.124 -0.9921 -0.003 0.0183>}
 dyn_backStrapIn_01_r {parent=dyn_backStrapIn_01_Base_r, loc=<0 0 0>, rot=<0 0 0 1>}
 dyn_backStrapIn_02_r {parent=dyn_backStrapIn_01_r, loc=<5.4161 0 0>, rot=<-0.0035 0.0157 -0.063 -0.9979>}
 dyn_backStrapOut_01_Base_r {parent=pelvis, loc=<4.7683 -12.0801 7.0565>, rot=<0.1252 -0.9918 -0.0039 0.0241>}
 dyn_backStrapOut_01_r {parent=dyn_backStrapOut_01_Base_r, loc=<0 0 0>, rot=<0 0 0 1>}
 dyn_backStrapOut_02_r {parent=dyn_backStrapOut_01_r, loc=<5.4806 0 0>, rot=<0.0019 -0.0089 -0.0727 -0.9973>}
 dyn_bigBeltMain_Base {parent=pelvis, loc=<1.3596 12.5599 0>, rot=<1 -0.0063 0 0>}
 dyn_bigBeltMain {parent=dyn_bigBeltMain_Base, loc=<0 0 0>, rot=<0 0 0 1>}
 dyn_bigBeltTop {parent=dyn_bigBeltMain, loc=<15.768 0 0>, rot=<0.9988 0.0488 0.0002 -0.006>}
 dyn_bigBeltTop_r {parent=dyn_bigBeltMain, loc=<15.8867 1.5911 -7.6217>, rot=<0.9644 0.0645 -0.0078 -0.2562>}
 dyn_bigBeltTop_l {parent=dyn_bigBeltMain, loc=<15.8404 1.9777 7.5513>, rot=<0.9885 0.0429 -0.0152 0.144>}
 dyn_chainFront_01_r {parent=dyn_bigBeltMain, loc=<2.118 4.6065 -11.9733>, rot=<-0.202 0.6101 0.7425 0.189>}
 dyn_chainFront_02_r {parent=dyn_chainFront_01_r, loc=<6.0839 0 0>, rot=<-0.1705 0.3552 0.0423 -0.9181>}
 dyn_chainFront_03_r {parent=dyn_chainFront_02_r, loc=<6.7094 0 0>, rot=<-0.0238 0.3388 -0.1179 -0.9332>}
 dyn_chainFront_04_r {parent=dyn_chainFront_03_r, loc=<6.6505 0 0>, rot=<-0.0119 0.3419 -0.1344 -0.93>}
 dyn_chainFront_01_l {parent=dyn_bigBeltMain, loc=<2.118 4.6065 11.9733>, rot=<-0.6101 -0.202 0.189 -0.7425>}
 dyn_chainFront_02_l {parent=dyn_chainFront_01_l, loc=<-6.0839 -0.0001 0>, rot=<-0.1705 0.3552 0.0423 -0.9181>}
 dyn_chainFront_03_l {parent=dyn_chainFront_02_l, loc=<-6.7094 0.0001 0>, rot=<-0.0238 0.3388 -0.1179 -0.9332>}
 dyn_chainFront_04_l {parent=dyn_chainFront_03_l, loc=<-6.6505 0 0>, rot=<-0.0119 0.3419 -0.1344 -0.93>}
 dyn_frontStrapOut_01_Base_l {parent=pelvis, loc=<-2.3163 10.209 -6.1835>, rot=<-0.007 -0.0004 0.9959 0.0903>}
 dyn_frontStrapOut_01_l {parent=dyn_frontStrapOut_01_Base_l, loc=<0 0 0>, rot=<0 0 0 1>}
 dyn_frontStrapOut_02_l {parent=dyn_frontStrapOut_01_l, loc=<6.8872 0 0>, rot=<0.0006 -0.0059 -0.0452 -0.999>}
 dyn_frontStrapOut_03_l {parent=dyn_frontStrapOut_02_l, loc=<7.3585 0 0>, rot=<0.0001 -0.0032 -0.021 -0.9998>}
 dyn_frontStrapOut_04_l {parent=dyn_frontStrapOut_03_l, loc=<9.0721 0 0>, rot=<0 0.0092 -0.0132 -0.9999>}
 dyn_frontStrapIn_01_Base_l {parent=pelvis, loc=<-3.0756 11.1657 -2.0541>, rot=<0.0009 0 0.9986 0.0532>}
 dyn_frontStrapIn_01_l {parent=dyn_frontStrapIn_01_Base_l, loc=<0 0 0>, rot=<0 0 0 1>}
 dyn_frontStrapIn_02_l {parent=dyn_frontStrapIn_01_l, loc=<6.7389 0 0>, rot=<0.0005 -0.0098 -0.0183 -0.9998>}
 dyn_frontStrapIn_03_l {parent=dyn_frontStrapIn_02_l, loc=<6.6037 0 0>, rot=<0.0001 -0.0021 -0.0069 -1>}
 dyn_frontStrapIn_04_l {parent=dyn_frontStrapIn_03_l, loc=<9.0682 0 0>, rot=<0 0.0024 -0.0113 -0.9999>}
 dyn_frontStrapIn_01_Base_r {parent=pelvis, loc=<-3.0756 11.1657 2.0541>, rot=<-0.0009 0 0.9986 0.0532>}
 dyn_frontStrapIn_01_r {parent=dyn_frontStrapIn_01_Base_r, loc=<0 0 0>, rot=<0 0 0 1>}
 dyn_frontStrapIn_02_r {parent=dyn_frontStrapIn_01_r, loc=<6.7389 0 0>, rot=<-0.0005 0.0098 -0.0183 -0.9998>}
 dyn_frontStrapIn_03_r {parent=dyn_frontStrapIn_02_r, loc=<6.6037 0 0>, rot=<-0.0001 0.0021 -0.0069 -1>}
 dyn_frontStrapIn_04_r {parent=dyn_frontStrapIn_03_r, loc=<9.0682 0 0>, rot=<0 -0.0024 -0.0113 -0.9999>}
 dyn_frontStrapOut_01_Base_r {parent=pelvis, loc=<-2.3163 10.209 6.1835>, rot=<0.007 0.0004 0.9959 0.0903>}
 dyn_frontStrapOut_01_r {parent=dyn_frontStrapOut_01_Base_r, loc=<0 0 0>, rot=<0 0 0 1>}
 dyn_frontStrapOut_02_r {parent=dyn_frontStrapOut_01_r, loc=<6.8872 0 0>, rot=<-0.0006 0.0059 -0.0452 -0.999>}
 dyn_frontStrapOut_03_r {parent=dyn_frontStrapOut_02_r, loc=<7.3585 0 0>, rot=<-0.0001 0.0032 -0.021 -0.9998>}
 dyn_frontStrapOut_04_r {parent=dyn_frontStrapOut_03_r, loc=<9.0721 0 0>, rot=<0 -0.0092 -0.0132 -0.9999>}
 dyn_loinClothFront_01_c_Base {parent=pelvis, loc=<-6.1035 11.1982 0.0322>, rot=<0 0 0.9998 0.0179>}
 dyn_loinClothFront_01_r {parent=dyn_loinClothFront_01_c_Base, loc=<3.0614 0.5088 5.5905>, rot=<0.0012 -0.025 0.0491 -0.9985>}
 dyn_loinClothFront_02_r {parent=dyn_loinClothFront_01_r, loc=<4.89 0 0>, rot=<-0.0011 0.0227 -0.0483 -0.9986>}
 dyn_loinClothFront_03_r {parent=dyn_loinClothFront_02_r, loc=<8.3684 0 0>, rot=<0.0006 -0.7055 -0.0006 -0.7088>}
 dyn_loinClothFront_01_l {parent=dyn_loinClothFront_01_c_Base, loc=<3.0645 0.7966 -5.5923>, rot=<-0.0013 0.023 0.0572 -0.9981>}
 dyn_loinClothFront_02_l {parent=dyn_loinClothFront_01_l, loc=<4.909 0 0>, rot=<0.0016 -0.0215 -0.0418 -0.9989>}
 dyn_loinClothFront_03_l {parent=dyn_loinClothFront_02_l, loc=<8.3794 0 0>, rot=<0.0109 -0.7081 -0.0109 -0.7059>}
 dyn_loinClothFront_02_c {parent=dyn_loinClothFront_01_c_Base, loc=<0 0 0>, rot=<-0.0242 -0.0003 0.0128 -0.9996>}
 dyn_loinClothFront_03_c {parent=dyn_loinClothFront_02_c, loc=<8.363 0 0>, rot=<0.0209 0.0007 0.0078 -0.9998>}
 dyn_loinClothFront_04_c {parent=dyn_loinClothFront_03_c, loc=<7.9175 0 0>, rot=<0.0142 0.0001 -0.0308 -0.9994>}
 dyn_loinClothFront_05_c {parent=dyn_loinClothFront_04_c, loc=<6.8601 0 0>, rot=<0.0073 -0.0003 -0.0259 -0.9996>}
 dyn_loinClothBack_01_Base {parent=pelvis, loc=<-15.7328 -15.3618 -0.0262>, rot=<-0.0272 -0.9996 0 0.0017>}
 dyn_loinClothBack_01 {parent=dyn_loinClothBack_01_Base, loc=<0 0 0>, rot=<0 0 0 1>}
 dyn_loinClothBack_02 {parent=dyn_loinClothBack_01, loc=<7.8997 0 0>, rot=<0 -0.001 0.0288 -0.9996>}
 dyn_loinClothBack_03 {parent=dyn_loinClothBack_02, loc=<6.4769 0 0>, rot=<0 0.0046 -0.0454 -0.999>}
 dyn_chainBack_01_Base_r {parent=pelvis, loc=<5.9706 -11.8693 9.3161>, rot=<0.1002 -0.7442 0.611 0.2507>}
 dyn_chainBack_01_r {parent=dyn_chainBack_01_Base_r, loc=<0 0 0>, rot=<0 0 0 1>}
 dyn_chainBack_02_r {parent=dyn_chainBack_01_r, loc=<10.0095 0 0>, rot=<0.0158 -0.3797 -0.0011 -0.925>}
 dyn_chainBack_03_r {parent=dyn_chainBack_02_r, loc=<8.8002 0 0>, rot=<0.0286 -0.3453 -0.1254 -0.9296>}
 dyn_chainBack_04_r {parent=dyn_chainBack_03_r, loc=<6.6505 0 0>, rot=<-0.4591 -0.3554 0.0487 -0.8128>}
 dyn_chainBack_01_Base_l {parent=pelvis, loc=<5.9706 -11.8693 -9.3161>, rot=<0.7442 0.1002 0.2507 -0.611>}
 dyn_chainBack_01_l {parent=dyn_chainBack_01_Base_l, loc=<0 0 0>, rot=<0 0 0 1>}
 dyn_chainBack_02_l {parent=dyn_chainBack_01_l, loc=<-10.0095 0 0>, rot=<0.0158 -0.3797 -0.0011 -0.925>}
 dyn_chainBack_03_l {parent=dyn_chainBack_02_l, loc=<-8.8002 0 0>, rot=<0.0286 -0.3453 -0.1254 -0.9296>}
 dyn_chainBack_04_l {parent=dyn_chainBack_03_l, loc=<-6.6505 0 0>, rot=<-0.4591 -0.3554 0.0487 -0.8128>}
 ik_foot_root {parent=root, loc=<0 0 0>, rot=<0 0 0 1>}
 ik_foot_l {parent=ik_foot_root, loc=<15.2608 -4.4176 9.4433>, rot=<-0.0357 -0.7045 -0.0284 0.7082>}
 ik_foot_r {parent=ik_foot_root, loc=<-15.2609 -4.4176 9.4431>, rot=<-0.7082 -0.0284 0.7045 -0.0357>}
 ik_hand_root {parent=root, loc=<0 0 0>, rot=<0 0 0 1>}
 ik_hand_gun {parent=ik_hand_root, loc=<-51.9729 4.0884 102.6935>, rot=<-0.6084 -0.1793 -0.4613 0.6205>}
 ik_hand_l {parent=ik_hand_gun, loc=<53.0286 82.1782 35.2051>, rot=<-0.5102 -0.7906 -0.3387 0>}
 ik_hand_r {parent=ik_hand_gun, loc=<0 0 0>, rot=<0 0 0 -1>}
]
variables [
 crouching {type=double, default=0}
 hanggliding {type=double, default=0}
 skydiving {type=double, default=0}
 targeting {type=double, default=0}
]
