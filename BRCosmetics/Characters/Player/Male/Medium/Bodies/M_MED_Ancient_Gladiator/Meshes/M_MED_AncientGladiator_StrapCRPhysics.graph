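# Physics asset for the Ancient Gladiator male medium cosmetic body, covering the front straps and loincloth.
# Four inner and outer strap capsule chains hang from the pelvis with tight swing limits, and their lower joints lock twist.
# The loincloth branches from one base into center, left and right chains; only the right has limited joints, while the others swing and twist freely.
# Pelvis and thigh shapes supply the body volumes around these chains.
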PhysicsAsset /BRCosmetics/Characters/Player/Male/Medium/Bodies/M_MED_Ancient_Gladiator/Meshes/M_MED_AncientGladiator_StrapCRPhysics
bodies {
 dyn_frontStrapIn_01_Base_l [{shapeType=Capsule, length=0.505, radius=0.505, rotation=<rot 0 -90 0>}]
 dyn_frontStrapIn_01_Base_r [{shapeType=Capsule, length=0.505, radius=0.505, rotation=<rot 0 -90 0>}]
 dyn_frontStrapIn_01_l [{shapeType=Capsule, length=5.9036, radius=1.8438, center=<3.6241 -1.3 0.0047>, rotation=<rot 90 13.2613 13.5434>}]
 dyn_frontStrapIn_01_r [{shapeType=Capsule, length=5.9036, radius=1.8438, center=<3.6241 -1.3 0.0047>, rotation=<rot 90 13.2613 13.5434>}]
 dyn_frontStrapIn_02_l [{shapeType=Capsule, length=6.7041, radius=1.8438, center=<3.6241 -1.3 0.0047>, rotation=<rot 90 -90 -89.718>}]
 dyn_frontStrapIn_02_r [{shapeType=Capsule, length=6.7041, radius=1.8438, center=<3.6241 -1.3 0.0047>, rotation=<rot 90 169.9218 170.2038>}]
 dyn_frontStrapIn_03_l [{shapeType=Capsule, length=7.4694, radius=1.8438, center=<3.6241 -1.3 0>, rotation=<rot 90 13.2613 13.5434>}]
 dyn_frontStrapIn_03_r [{shapeType=Capsule, length=7.4694, radius=1.8438, center=<3.6241 -1.3 0>, rotation=<rot 90 -13.2613 -12.9793>}]
 dyn_frontStrapIn_04_l [{shapeType=Capsule, length=11.3237, radius=1.8438, center=<3.5581 -1.2996 0>, rotation=<rot 90 -90 -89.718>}]
 dyn_frontStrapIn_04_r [{shapeType=Capsule, length=11.1516, radius=1.8438, center=<3.6689 -1.3002 0>, rotation=<rot 90 -13.2613 -12.9794>}]
 dyn_frontStrapOut_01_Base_l [{shapeType=Capsule, length=0.505, radius=0.505, rotation=<rot 0 -90 0>}]
 dyn_frontStrapOut_01_Base_r [{shapeType=Capsule, length=0.505, radius=0.505, rotation=<rot 0 -90 0>}]
 dyn_frontStrapOut_01_l [{shapeType=Capsule, length=5.9036, radius=1.8438, center=<3.6241 -1.3 0.0047>, rotation=<rot 90 169.9218 170.2038>}]
 dyn_frontStrapOut_01_r [{shapeType=Capsule, length=5.9036, radius=1.8438, center=<3.6241 -1.3 0.0048>, rotation=<rot 90 0 0.282>}]
 dyn_frontStrapOut_02_l [{shapeType=Capsule, length=5.9036, radius=1.8438, center=<3.6241 -1.3 0.0047>, rotation=<rot 90 0 0.282>}]
 dyn_frontStrapOut_02_r [{shapeType=Capsule, length=5.9036, radius=1.8438, center=<3.6241 -1.3 0.0047>, rotation=<rot 90 169.9218 170.2038>}]
 dyn_frontStrapOut_03_l [{shapeType=Capsule, length=5.9036, radius=1.8438, center=<3.6241 -1.3 0>, rotation=<rot 90 0 0.282>}]
 dyn_frontStrapOut_03_r [{shapeType=Capsule, length=5.9036, radius=1.8438, center=<3.6241 -1.3 0>, rotation=<rot 90 0 0.282>}]
 dyn_frontStrapOut_04_l [{shapeType=Capsule, length=5.9036, radius=1.8438, center=<3.6241 -1.3 0>, rotation=<rot 90 0 0.282>}]
 dyn_frontStrapOut_04_r [{shapeType=Capsule, length=5.9036, radius=1.8438, center=<3.6241 -1.3 0>, rotation=<rot 90 0 0.282>}]
 dyn_loinClothFront_01_c_Base [{shapeType=Box, extentX=1.01, extentY=1.01, extentZ=1.01, center=<0 1.7596 0>, rotation=<rot 0 90 0>}]
 dyn_loinClothFront_01_r [{shapeType=Box, extentX=9.918, extentY=11.7619, extentZ=11.6628, center=<0.4677 4.8309 0.0002>, rotation=<rot 0 90 0>}]
 dyn_loinClothFront_02_c [{shapeType=Capsule, length=10.3165, radius=5.8156, center=<3.389 5.2973 0>}]
 dyn_loinClothFront_02_l [{shapeType=Capsule, length=7.1814, radius=4.8238, center=<4.9056 5.4519 0>}]
 dyn_loinClothFront_02_r [{shapeType=Box, extentX=8.7546, extentY=8.574, extentZ=12.3882, center=<4.2526 4.831 0.0002>, rotation=<rot 0 90 0>}]
 dyn_loinClothFront_03_c [{shapeType=Capsule, length=6.5331, radius=7.0842, center=<3.7974 6.1273 -0.0001>}]
 dyn_loinClothFront_03_l [{shapeType=Capsule, length=7.1814, radius=4.8238, center=<0 5.2071 4.9056>, rotation=<rot 90 0 0>}]
 dyn_loinClothFront_03_r [{shapeType=Box, extentX=8.2143, extentY=9.0023, extentZ=9.5103, center=<0 4.8309 3.8565>, rotation=<rot 0 90 0>}]
 dyn_loinClothFront_04_c [{shapeType=Box, extentX=8.2143, extentY=10.1293, extentZ=10.087, center=<3.5307 4.8307 0.0002>, rotation=<rot 0 90 0>}]
 dyn_loinClothFront_05_c [{shapeType=Box, extentX=8.2143, extentY=9.9903, extentZ=10.2639, center=<0 4.8308 0.0002>, rotation=<rot 0 90 0>}]
 pelvis [{shapeType=Capsule, length=10.6327, radius=10.7063, center=<-13.3244 0.9524 0.0002>}]
 thigh_l [
  {shapeType=Sphere, radius=10.817, center=<-14.9086 -2.696 4.7862>}
  {shapeType=Capsule, length=23.0647, radius=9.7066, center=<-15.8027 -0.0001 0.558>, rotation=<rot -90 0 0>}
 ]
 thigh_r [{shapeType=Capsule, length=23.3359, radius=9.379, center=<16.5241 0.0007 0.5714>, rotation=<rot 90 0 0>}]
}
constraints [
 {
  bone1Name=dyn_loinClothFront_01_c_Base
  bone2Name=pelvis
  swing1LimitDegrees=45
  swing1Motion=Free
  swing2LimitDegrees=45
  swing2Motion=Free
  twistLimitDegrees=45
  twistMotion=Free
 }
 {bone1Name=dyn_loinClothFront_01_r, bone2Name=dyn_loinClothFront_01_c_Base, swing1LimitDegrees=18, swing2LimitDegrees=14, twistLimitDegrees=6}
 {bone1Name=dyn_loinClothFront_02_r, bone2Name=dyn_loinClothFront_01_r, swing1LimitDegrees=18, swing2LimitDegrees=14, twistLimitDegrees=6}
 {bone1Name=dyn_loinClothFront_03_r, bone2Name=dyn_loinClothFront_02_r, swing1LimitDegrees=6, swing2LimitDegrees=14, twistLimitDegrees=18}
 {
  bone1Name=dyn_loinClothFront_02_c
  bone2Name=dyn_loinClothFront_01_c_Base
  swing1LimitDegrees=45
  swing1Motion=Free
  swing2LimitDegrees=45
  swing2Motion=Free
  twistLimitDegrees=45
  twistMotion=Free
 }
 {
  bone1Name=dyn_loinClothFront_03_c
  bone2Name=dyn_loinClothFront_02_c
  swing1LimitDegrees=45
  swing1Motion=Free
  swing2LimitDegrees=45
  swing2Motion=Free
  twistLimitDegrees=45
  twistMotion=Free
 }
 {
  bone1Name=dyn_loinClothFront_04_c
  bone2Name=dyn_loinClothFront_03_c
  swing1LimitDegrees=45
  swing1Motion=Free
  swing2LimitDegrees=45
  swing2Motion=Free
  twistLimitDegrees=45
  twistMotion=Free
 }
 {
  bone1Name=dyn_loinClothFront_05_c
  bone2Name=dyn_loinClothFront_04_c
  swing1LimitDegrees=45
  swing1Motion=Free
  swing2LimitDegrees=45
  swing2Motion=Free
  twistLimitDegrees=45
  twistMotion=Free
 }
 {
  bone1Name=thigh_r
  bone2Name=pelvis
  swing1LimitDegrees=45
  swing1Motion=Free
  swing2LimitDegrees=45
  swing2Motion=Free
  twistLimitDegrees=45
  twistMotion=Free
 }
 {
  bone1Name=thigh_l
  bone2Name=pelvis
  swing1LimitDegrees=45
  swing1Motion=Free
  swing2LimitDegrees=45
  swing2Motion=Free
  twistLimitDegrees=45
  twistMotion=Free
 }
 {
  bone1Name=dyn_frontStrapOut_01_Base_l
  bone2Name=pelvis
  swing1LimitDegrees=45
  swing1Motion=Free
  swing2LimitDegrees=45
  swing2Motion=Free
  twistLimitDegrees=45
  twistMotion=Free
 }
 {bone1Name=dyn_frontStrapOut_01_l, bone2Name=dyn_frontStrapOut_01_Base_l, swing1LimitDegrees=16, swing2LimitDegrees=12, twistLimitDegrees=4}
 {bone1Name=dyn_frontStrapOut_02_l, bone2Name=dyn_frontStrapOut_01_l, swing1LimitDegrees=16, swing2LimitDegrees=12, twistLimitDegrees=4}
 {
  bone1Name=dyn_frontStrapIn_01_Base_l
  bone2Name=pelvis
  swing1LimitDegrees=45
  swing1Motion=Free
  swing2LimitDegrees=45
  swing2Motion=Free
  twistLimitDegrees=45
  twistMotion=Free
 }
 {bone1Name=dyn_frontStrapIn_01_l, bone2Name=dyn_frontStrapIn_01_Base_l, swing1LimitDegrees=16, swing2LimitDegrees=12, twistLimitDegrees=4}
 {bone1Name=dyn_frontStrapIn_02_l, bone2Name=dyn_frontStrapIn_01_l, swing1LimitDegrees=16, swing2LimitDegrees=12, twistLimitDegrees=4}
 {
  bone1Name=dyn_frontStrapIn_01_Base_r
  bone2Name=pelvis
  swing1LimitDegrees=45
  swing1Motion=Free
  swing2LimitDegrees=45
  swing2Motion=Free
  twistLimitDegrees=45
  twistMotion=Free
 }
 {bone1Name=dyn_frontStrapIn_01_r, bone2Name=dyn_frontStrapIn_01_Base_r, swing1LimitDegrees=16, swing2LimitDegrees=12, twistLimitDegrees=4}
 {bone1Name=dyn_frontStrapIn_02_r, bone2Name=dyn_frontStrapIn_01_r, swing1LimitDegrees=16, swing2LimitDegrees=12, twistLimitDegrees=4}
 {
  bone1Name=dyn_frontStrapOut_01_Base_r
  bone2Name=pelvis
  swing1LimitDegrees=45
  swing1Motion=Free
  swing2LimitDegrees=45
  swing2Motion=Free
  twistLimitDegrees=45
  twistMotion=Free
 }
 {bone1Name=dyn_frontStrapOut_01_r, bone2Name=dyn_frontStrapOut_01_Base_r, swing1LimitDegrees=16, swing2LimitDegrees=12, twistLimitDegrees=4}
 {bone1Name=dyn_frontStrapOut_02_r, bone2Name=dyn_frontStrapOut_01_r, swing1LimitDegrees=16, swing2LimitDegrees=12, twistLimitDegrees=4}
 {
  bone1Name=dyn_frontStrapOut_03_l
  bone2Name=dyn_frontStrapOut_02_l
  swing1LimitDegrees=16
  swing2LimitDegrees=12
  twistLimitDegrees=4
  twistMotion=Locked
 }
 {
  bone1Name=dyn_frontStrapOut_04_l
  bone2Name=dyn_frontStrapOut_03_l
  swing1LimitDegrees=16
  swing2LimitDegrees=12
  twistLimitDegrees=4
  twistMotion=Locked
 }
 {
  bone1Name=dyn_frontStrapOut_03_r
  bone2Name=dyn_frontStrapOut_02_r
  swing1LimitDegrees=16
  swing2LimitDegrees=12
  twistLimitDegrees=4
  twistMotion=Locked
 }
 {
  bone1Name=dyn_frontStrapOut_04_r
  bone2Name=dyn_frontStrapOut_03_r
  swing1LimitDegrees=16
  swing2LimitDegrees=12
  twistLimitDegrees=4
  twistMotion=Locked
 }
 {
  bone1Name=dyn_frontStrapIn_03_l
  bone2Name=dyn_frontStrapIn_02_l
  swing1LimitDegrees=16
  swing2LimitDegrees=12
  twistLimitDegrees=4
  twistMotion=Locked
 }
 {
  bone1Name=dyn_frontStrapIn_04_l
  bone2Name=dyn_frontStrapIn_03_l
  swing1LimitDegrees=16
  swing2LimitDegrees=12
  twistLimitDegrees=4
  twistMotion=Locked
 }
 {
  bone1Name=dyn_frontStrapIn_03_r
  bone2Name=dyn_frontStrapIn_02_r
  swing1LimitDegrees=16
  swing2LimitDegrees=12
  twistLimitDegrees=4
  twistMotion=Locked
 }
 {
  bone1Name=dyn_frontStrapIn_04_r
  bone2Name=dyn_frontStrapIn_03_r
  swing1LimitDegrees=16
  swing2LimitDegrees=12
  twistLimitDegrees=4
  twistMotion=Locked
 }
 {
  bone1Name=dyn_loinClothFront_03_l
  bone2Name=dyn_loinClothFront_02_l
  swing1LimitDegrees=45
  swing1Motion=Free
  swing2LimitDegrees=45
  swing2Motion=Free
  twistLimitDegrees=45
  twistMotion=Free
 }
 {
  bone1Name=dyn_loinClothFront_02_l
  bone2Name=dyn_loinClothFront_01_c_Base
  swing1LimitDegrees=45
  swing1Motion=Free
  swing2LimitDegrees=45
  swing2Motion=Free
  twistLimitDegrees=45
  twistMotion=Free
 }
]
