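# Physics asset for the Ancient Gladiator medium male body's waist chains, named as the crouch variant.
# Front chains hang from a belt bone and back chains from pelvis-linked base bones, as capsule segments with narrow swing and twist limits.
# Free-angle cross constraints tie overlapping front and back segments in reverse order on each side, joining them into loops.
# Pelvis and double-capsule thigh bodies supply the leg and hip shapes around the chains.
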PhysicsAsset /BRCosmetics/Characters/Player/Male/Medium/Bodies/M_MED_Ancient_Gladiator/Meshes/M_MED_Ancient_Gladiator_Chain_Crouch_Physics
bodies {
 dyn_bigBeltMain [{shapeType=Box, extentX=0.2, extentY=0.2, extentZ=0.2}]
 dyn_bigBeltMain_Base [{shapeType=Box, extentX=0.2, extentY=0.2, extentZ=0.2}]
 dyn_chainBack_01_Base_l [{shapeType=Capsule, length=0.505, radius=0.505, rotation=<rot 0 -90 180>}]
 dyn_chainBack_01_Base_r [{shapeType=Capsule, length=0.505, radius=0.505, rotation=<rot 0 -90 0>}]
 dyn_chainBack_01_l [{shapeType=Capsule, length=9.1087, radius=1.9926, center=<-5.2066 0.0002 0>, rotation=<rot 90 0 180>}]
 dyn_chainBack_01_r [{shapeType=Capsule, length=9.1087, radius=1.9926, center=<5.2066 -0.0002 0>, rotation=<rot 90 0 0>}]
 dyn_chainBack_02_l [{shapeType=Capsule, length=7.9289, radius=1.7603, center=<-4.5014 0.0002 0>, rotation=<rot 90 0 180>}]
 dyn_chainBack_02_r [{shapeType=Capsule, length=7.9289, radius=1.7603, center=<4.5014 -0.0002 0>, rotation=<rot 90 0 0>}]
 dyn_chainBack_03_l [{shapeType=Capsule, length=6.4779, radius=1.708, center=<-3.3378 0.0002 0>, rotation=<rot 90 0 180>}]
 dyn_chainBack_03_r [{shapeType=Capsule, length=6.4779, radius=1.708, center=<3.3378 -0.0002 0>, rotation=<rot 90 0 0>}]
 dyn_chainBack_04_l [{shapeType=Capsule, length=5.8501, radius=1.6574, center=<-3.5782 0.0002 0>, rotation=<rot 90 0 180>}]
 dyn_chainBack_04_r [{shapeType=Capsule, length=6.4341, radius=1.6574, center=<3.5782 -0.0002 0>, rotation=<rot 90 0 0>}]
 dyn_chainFront_01_l [{shapeType=Capsule, length=5.5813, radius=1.7603, center=<-3.0594 0.0002 0>, rotation=<rot 90 0 180>}]
 dyn_chainFront_01_r [{shapeType=Capsule, length=5.5813, radius=1.7603, center=<3.0594 -0.0002 0>, rotation=<rot 90 0 0>}]
 dyn_chainFront_02_l [{shapeType=Capsule, length=4.7621, radius=2.2603, center=<-3.2376 0.0002 0>, rotation=<rot 90 0 180>}]
 dyn_chainFront_02_r [{shapeType=Capsule, length=6.3876, radius=1.7603, center=<3.2376 -0.0002 0>, rotation=<rot 90 0 0>}]
 dyn_chainFront_03_l [{shapeType=Capsule, length=5.1127, radius=2.0335, center=<-3.2566 0.0001 0>, rotation=<rot 90 0 180>}]
 dyn_chainFront_03_r [{shapeType=Capsule, length=5.6127, radius=1.5335, center=<3.2566 -0.0001 0>, rotation=<rot 90 0 0>}]
 dyn_chainFront_04_l [{shapeType=Capsule, length=7.4289, radius=2.2079, center=<-4.5014 0.0002 0>, rotation=<rot 90 0 180>}]
 dyn_chainFront_04_r [{shapeType=Capsule, length=7.9289, radius=1.7079, center=<4.5014 -0.0002 0>, rotation=<rot 90 0 0>}]
 pelvis [{shapeType=Capsule, length=10.1327, radius=11.2063, center=<-5.6844 -2.986 0>}]
 thigh_l [
  {shapeType=Capsule, length=23.9255, radius=8.7894, center=<-16.5241 -0.0007 -0.0869>, rotation=<rot 90 0 180>}
  {shapeType=Capsule, length=22.6213, radius=7.3787, center=<-21.2076 -7.0753 5.8751>, rotation=<rot 86.6133 0 178.1214>}
 ]
 thigh_r [
  {shapeType=Capsule, length=23.9255, radius=8.7894, center=<16.5241 0.0007 0.0869>, rotation=<rot 90 0 0>}
  {shapeType=Capsule, length=22.6213, radius=7.3787, center=<21.2077 7.0753 -5.8727>, rotation=<rot 86.6143 0 -1.8786>}
 ]
}
constraints [
 {
  bone1Name=thigh_r
  bone2Name=pelvis
  swing1LimitDegrees=45
  swing1Motion=Free
  swing2LimitDegrees=45
  swing2Motion=Free
  twistLimitDegrees=45
  twistMotion=Free
 }
 {
  bone1Name=dyn_bigBeltMain_Base
  bone2Name=pelvis
  swing1LimitDegrees=45
  swing1Motion=Free
  swing2LimitDegrees=45
  swing2Motion=Free
  twistLimitDegrees=45
  twistMotion=Free
 }
 {
  bone1Name=dyn_bigBeltMain
  bone2Name=dyn_bigBeltMain_Base
  swing1LimitDegrees=45
  swing1Motion=Free
  swing2LimitDegrees=45
  swing2Motion=Free
  twistLimitDegrees=45
  twistMotion=Free
 }
 {bone1Name=dyn_chainFront_01_r, bone2Name=dyn_bigBeltMain, swing1LimitDegrees=18, swing2LimitDegrees=12, twistLimitDegrees=4}
 {bone1Name=dyn_chainFront_02_r, bone2Name=dyn_chainFront_01_r, swing1LimitDegrees=18, swing2LimitDegrees=12, twistLimitDegrees=4}
 {bone1Name=dyn_chainFront_03_r, bone2Name=dyn_chainFront_02_r, swing1LimitDegrees=18, swing2LimitDegrees=12, twistLimitDegrees=4}
 {bone1Name=dyn_chainFront_04_r, bone2Name=dyn_chainFront_03_r, swing1LimitDegrees=18, swing2LimitDegrees=12, twistLimitDegrees=4}
 {
  bone1Name=dyn_chainBack_01_Base_r
  bone2Name=pelvis
  swing1LimitDegrees=45
  swing1Motion=Free
  swing2LimitDegrees=45
  swing2Motion=Free
  twistLimitDegrees=45
  twistMotion=Free
 }
 {bone1Name=dyn_chainBack_01_r, bone2Name=dyn_chainBack_01_Base_r, swing1LimitDegrees=18, swing2LimitDegrees=12, twistLimitDegrees=4}
 {bone1Name=dyn_chainBack_02_r, bone2Name=dyn_chainBack_01_r, swing1LimitDegrees=18, swing2LimitDegrees=12, twistLimitDegrees=4}
 {bone1Name=dyn_chainBack_03_r, bone2Name=dyn_chainBack_02_r, swing1LimitDegrees=18, swing2LimitDegrees=12, twistLimitDegrees=4}
 {bone1Name=dyn_chainBack_04_r, bone2Name=dyn_chainBack_03_r, swing1LimitDegrees=18, swing2LimitDegrees=12, twistLimitDegrees=4}
 {
  bone1Name=dyn_chainBack_03_r
  bone2Name=dyn_chainFront_03_r
  swing1LimitDegrees=45
  swing1Motion=Free
  swing2LimitDegrees=45
  swing2Motion=Free
  twistLimitDegrees=45
  twistMotion=Free
 }
 {
  bone1Name=dyn_chainBack_02_r
  bone2Name=dyn_chainFront_04_r
  swing1LimitDegrees=45
  swing1Motion=Free
  swing2LimitDegrees=45
  swing2Motion=Free
  twistLimitDegrees=45
  twistMotion=Free
 }
 {
  bone1Name=dyn_chainFront_02_r
  bone2Name=dyn_chainBack_04_r
  swing1LimitDegrees=45
  swing1Motion=Free
  swing2LimitDegrees=45
  swing2Motion=Free
  twistLimitDegrees=45
  twistMotion=Free
 }
 {bone1Name=dyn_chainFront_01_l, bone2Name=dyn_bigBeltMain, swing1LimitDegrees=12, swing2LimitDegrees=12, twistLimitDegrees=4}
 {bone1Name=dyn_chainFront_02_l, bone2Name=dyn_chainFront_01_l, swing1LimitDegrees=12, swing2LimitDegrees=12, twistLimitDegrees=4}
 {bone1Name=dyn_chainFront_03_l, bone2Name=dyn_chainFront_02_l, swing1LimitDegrees=12, swing2LimitDegrees=12, twistLimitDegrees=4}
 {bone1Name=dyn_chainFront_04_l, bone2Name=dyn_chainFront_03_l, swing1LimitDegrees=12, swing2LimitDegrees=12, twistLimitDegrees=4}
 {
  bone1Name=dyn_chainBack_01_Base_l
  bone2Name=pelvis
  swing1LimitDegrees=45
  swing1Motion=Free
  swing2LimitDegrees=45
  swing2Motion=Free
  twistLimitDegrees=45
  twistMotion=Free
 }
 {bone1Name=dyn_chainBack_01_l, bone2Name=dyn_chainBack_01_Base_l, swing1LimitDegrees=12, swing2LimitDegrees=12, twistLimitDegrees=4}
 {bone1Name=dyn_chainBack_02_l, bone2Name=dyn_chainBack_01_l, swing1LimitDegrees=12, swing2LimitDegrees=12, twistLimitDegrees=4}
 {bone1Name=dyn_chainBack_03_l, bone2Name=dyn_chainBack_02_l, swing1LimitDegrees=12, swing2LimitDegrees=12, twistLimitDegrees=4}
 {bone1Name=dyn_chainBack_04_l, bone2Name=dyn_chainBack_03_l, swing1LimitDegrees=12, swing2LimitDegrees=12, twistLimitDegrees=4}
 {
  bone1Name=dyn_chainBack_04_l
  bone2Name=dyn_chainFront_02_l
  swing1LimitDegrees=45
  swing1Motion=Free
  swing2LimitDegrees=45
  swing2Motion=Free
  twistLimitDegrees=45
  twistMotion=Free
 }
 {
  bone1Name=dyn_chainBack_03_l
  bone2Name=dyn_chainFront_03_l
  swing1LimitDegrees=45
  swing1Motion=Free
  swing2LimitDegrees=45
  swing2Motion=Free
  twistLimitDegrees=45
  twistMotion=Free
 }
 {
  bone1Name=dyn_chainBack_02_l
  bone2Name=dyn_chainFront_04_l
  swing1LimitDegrees=45
  swing1Motion=Free
  swing2LimitDegrees=45
  swing2Motion=Free
  twistLimitDegrees=45
  twistMotion=Free
 }
 {
  bone1Name=thigh_l
  bone2Name=pelvis
  swing1LimitDegrees=45
  swing1Motion=Free
  swing2LimitDegrees=45
  swing2Motion=Free
  twistLimitDegrees=45
  twistMotion=Free
 }
]
