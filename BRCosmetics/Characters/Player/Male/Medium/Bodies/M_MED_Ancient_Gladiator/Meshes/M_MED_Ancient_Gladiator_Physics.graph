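# Physics asset for the Ancient Gladiator male medium body cosmetic, simulating its dangling loincloth, back straps and belt.
# Capsules on the pelvis and thighs give the body's collision shapes, while box bodies cover the dyn_ cloth bones.
# Each chain hangs from a small base body jointed to the pelvis with free angular motion.
# Tight swing and twist limits govern most links below, except the belt and front-center loincloth's first link, which stay free.
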PhysicsAsset /BRCosmetics/Characters/Player/Male/Medium/Bodies/M_MED_Ancient_Gladiator/Meshes/M_MED_Ancient_Gladiator_Physics
bodies {
 dyn_backStrapIn_01_Base_l [{shapeType=Box, extentX=1.01, extentY=1.01, extentZ=1.01, rotation=<rot 0 -90 0>}]
 dyn_backStrapIn_01_Base_r [{shapeType=Box, extentX=1.01, extentY=1.01, extentZ=1.01, rotation=<rot 0 -90 0>}]
 dyn_backStrapIn_01_l [{shapeType=Box, extentX=0.7729, extentY=5.3514, extentZ=4.5901, center=<3 0 0>, rotation=<rot 0 90 0>}]
 dyn_backStrapIn_01_r [{shapeType=Box, extentX=0.8148, extentY=5.357, extentZ=4.5938, center=<3 0 0>, rotation=<rot 0 90 0>}]
 dyn_backStrapIn_02_l [{shapeType=Box, extentX=1.4104, extentY=10.7368, extentZ=4.9555, center=<4.9319 0.0001 0>, rotation=<rot 0 90 0>}]
 dyn_backStrapIn_02_r [{shapeType=Box, extentX=1.4366, extentY=10.9472, extentZ=5.1035, center=<4.9319 0.0001 0>, rotation=<rot 0 90 0>}]
 dyn_backStrapOut_01_Base_l [{shapeType=Box, extentX=1.01, extentY=1.01, extentZ=1.01, rotation=<rot 0 -90 0>}]
 dyn_backStrapOut_01_Base_r [{shapeType=Box, extentX=1.01, extentY=1.01, extentZ=1.01, rotation=<rot 0 -90 0>}]
 dyn_backStrapOut_01_l [{shapeType=Box, extentX=0.693, extentY=5.5798, extentZ=3.9123, center=<3 0 0>, rotation=<rot 0 90 0>}]
 dyn_backStrapOut_01_r [{shapeType=Box, extentX=0.7343, extentY=5.5796, extentZ=3.9123, center=<3 0 0>, rotation=<rot 0 90 0>}]
 dyn_backStrapOut_02_l [{shapeType=Box, extentX=1.1876, extentY=7.7378, extentZ=3.9123, center=<3.8167 0 0>, rotation=<rot 0 90 0>}]
 dyn_backStrapOut_02_r [{shapeType=Box, extentX=1.1674, extentY=7.7662, extentZ=3.9123, center=<3.8167 0 0>, rotation=<rot 0 90 0>}]
 dyn_bigBeltMain [{shapeType=Box, extentX=0.2, extentY=0.2, extentZ=0.2}]
 dyn_bigBeltMain_Base [{shapeType=Box, extentX=0.2, extentY=0.2, extentZ=0.2}]
 dyn_loinClothBack_01 [{shapeType=Box, extentX=3.1431, extentY=9.134, extentZ=20.5999, center=<3.5044 0.0001 0>, rotation=<rot 0 90 0>}]
 dyn_loinClothBack_01_Base [{shapeType=Box, extentX=1.01, extentY=1.01, extentZ=1.01, rotation=<rot 0 90 0>}]
 dyn_loinClothBack_02 [{shapeType=Box, extentX=1.9797, extentY=7.688, extentZ=19.0518, center=<3.2359 0.0002 0>, rotation=<rot 0 90 0>}]
 dyn_loinClothBack_03 [{shapeType=Box, extentX=1.4395, extentY=11.0094, extentZ=10.529, center=<5.0291 0.0001 0>, rotation=<rot 0 90 0>}]
 dyn_loinClothFront_01_c_Base [{shapeType=Box, extentX=1.01, extentY=1.01, extentZ=1.01, rotation=<rot 0 90 0>}]
 dyn_loinClothFront_01_l [{shapeType=Box, extentX=3.1431, extentY=6.5164, extentZ=11.6628, center=<2.4563 -0.0001 0>, rotation=<rot 0 90 0>}]
 dyn_loinClothFront_01_r [{shapeType=Box, extentX=3.1431, extentY=6.8068, extentZ=11.6628, center=<2.6716 0.0001 0>, rotation=<rot 0 90 0>}]
 dyn_loinClothFront_02_c [{shapeType=Box, extentX=3.1431, extentY=10.4955, extentZ=12.3995, center=<3.6349 0.0001 0>, rotation=<rot 0 90 0>}]
 dyn_loinClothFront_02_l [{shapeType=Box, extentX=1.9797, extentY=9.209, extentZ=12.3882, center=<4.2519 0 0>, rotation=<rot 0 90 0>}]
 dyn_loinClothFront_02_r [{shapeType=Box, extentX=1.9797, extentY=8.574, extentZ=12.3882, center=<4.2526 0.0001 0>, rotation=<rot 0 90 0>}]
 dyn_loinClothFront_03_c [{shapeType=Box, extentX=1.9797, extentY=8.9499, extentZ=10.8514, center=<3.7448 0.0003 0>, rotation=<rot 0 90 0>}]
 dyn_loinClothFront_03_l [{shapeType=Box, extentX=1.4395, extentY=9.0023, extentZ=9.5103, center=<0 0.0001 3.8379>, rotation=<rot 0 90 0>}]
 dyn_loinClothFront_03_r [{shapeType=Box, extentX=1.4395, extentY=9.0023, extentZ=9.5103, center=<0 0.0001 3.8564>, rotation=<rot 0 90 0>}]
 dyn_loinClothFront_04_c [{shapeType=Box, extentX=1.4395, extentY=8.2893, extentZ=8.5764, center=<3.4763 -0.0001 0>, rotation=<rot 0 90 0>}]
 dyn_loinClothFront_05_c [{shapeType=Box, extentX=1.4395, extentY=9.9903, extentZ=8.0262, center=<3.835 0 -0.0001>, rotation=<rot 0 90 0>}]
 pelvis [{shapeType=Capsule, length=10.1327, radius=11.2063, center=<-5.6844 -2.986 0>}]
 thigh_l [
  {shapeType=Capsule, length=23.9255, radius=8.7894, center=<-16.5241 -0.0007 -0.0869>, rotation=<rot 90 0 180>}
  {shapeType=Capsule, length=22.6213, radius=7.3787, center=<-21.2076 -7.0753 5.8751>, rotation=<rot 86.6133 0 178.1214>}
 ]
 thigh_r [
  {shapeType=Capsule, length=23.9255, radius=8.7894, center=<16.5241 0.0007 0.0869>, rotation=<rot 90 0 0>}
  {shapeType=Capsule, length=22.6213, radius=7.3787, center=<21.2077 7.0753 -5.8727>, rotation=<rot 86.6143 0 -1.8786>}
 ]
}
constraints [
 {
  bone1Name=dyn_loinClothBack_01_Base
  bone2Name=pelvis
  swing1LimitDegrees=45
  swing1Motion=Free
  swing2LimitDegrees=45
  swing2Motion=Free
  twistLimitDegrees=45
  twistMotion=Free
 }
 {bone1Name=dyn_loinClothBack_01, bone2Name=dyn_loinClothBack_01_Base, swing1LimitDegrees=20, swing2LimitDegrees=6, twistLimitDegrees=4}
 {bone1Name=dyn_loinClothBack_02, bone2Name=dyn_loinClothBack_01, swing1LimitDegrees=20, swing2LimitDegrees=6, twistLimitDegrees=4}
 {bone1Name=dyn_loinClothBack_03, bone2Name=dyn_loinClothBack_02, swing1LimitDegrees=20, swing2LimitDegrees=6, twistLimitDegrees=4}
 {
  bone1Name=dyn_loinClothFront_01_c_Base
  bone2Name=pelvis
  swing1LimitDegrees=45
  swing1Motion=Free
  swing2LimitDegrees=45
  swing2Motion=Free
  twistLimitDegrees=45
  twistMotion=Free
 }
 {bone1Name=dyn_loinClothFront_01_r, bone2Name=dyn_loinClothFront_01_c_Base, swing1LimitDegrees=18, swing2LimitDegrees=14, twistLimitDegrees=6}
 {bone1Name=dyn_loinClothFront_02_r, bone2Name=dyn_loinClothFront_01_r, swing1LimitDegrees=18, swing2LimitDegrees=14, twistLimitDegrees=6}
 {bone1Name=dyn_loinClothFront_03_r, bone2Name=dyn_loinClothFront_02_r, swing1LimitDegrees=6, swing2LimitDegrees=14, twistLimitDegrees=18}
 {
  bone1Name=dyn_loinClothFront_01_l
  bone2Name=dyn_loinClothFront_01_c_Base
  swing1LimitDegrees=18
  swing2LimitDegrees=14
  twistLimitDegrees=18
 }
 {bone1Name=dyn_loinClothFront_02_l, bone2Name=dyn_loinClothFront_01_l, swing1LimitDegrees=18, swing2LimitDegrees=14, twistLimitDegrees=18}
 {bone1Name=dyn_loinClothFront_03_l, bone2Name=dyn_loinClothFront_02_l, swing1LimitDegrees=18, swing2LimitDegrees=14, twistLimitDegrees=18}
 {
  bone1Name=dyn_loinClothFront_02_c
  bone2Name=dyn_loinClothFront_01_c_Base
  swing1LimitDegrees=45
  swing1Motion=Free
  swing2LimitDegrees=45
  swing2Motion=Free
  twistLimitDegrees=45
  twistMotion=Free
 }
 {bone1Name=dyn_loinClothFront_03_c, bone2Name=dyn_loinClothFront_02_c, swing1LimitDegrees=18, swing2LimitDegrees=14, twistLimitDegrees=6}
 {bone1Name=dyn_loinClothFront_04_c, bone2Name=dyn_loinClothFront_03_c, swing1LimitDegrees=18, swing2LimitDegrees=14, twistLimitDegrees=6}
 {bone1Name=dyn_loinClothFront_05_c, bone2Name=dyn_loinClothFront_04_c, swing1LimitDegrees=18, swing2LimitDegrees=8, twistLimitDegrees=6}
 {
  bone1Name=thigh_r
  bone2Name=pelvis
  swing1LimitDegrees=45
  swing1Motion=Free
  swing2LimitDegrees=45
  swing2Motion=Free
  twistLimitDegrees=45
  twistMotion=Free
 }
 {
  bone1Name=dyn_backStrapOut_01_Base_l
  bone2Name=pelvis
  swing1LimitDegrees=45
  swing1Motion=Free
  swing2LimitDegrees=45
  swing2Motion=Free
  twistLimitDegrees=45
  twistMotion=Free
 }
 {bone1Name=dyn_backStrapOut_01_l, bone2Name=dyn_backStrapOut_01_Base_l, swing1LimitDegrees=18, swing2LimitDegrees=6, twistLimitDegrees=2}
 {bone1Name=dyn_backStrapOut_02_l, bone2Name=dyn_backStrapOut_01_l, swing1LimitDegrees=18, swing2LimitDegrees=6, twistLimitDegrees=2}
 {
  bone1Name=dyn_backStrapIn_01_Base_l
  bone2Name=pelvis
  swing1LimitDegrees=45
  swing1Motion=Free
  swing2LimitDegrees=45
  swing2Motion=Free
  twistLimitDegrees=45
  twistMotion=Free
 }
 {bone1Name=dyn_backStrapIn_01_l, bone2Name=dyn_backStrapIn_01_Base_l, swing1LimitDegrees=18, swing2LimitDegrees=6, twistLimitDegrees=2}
 {bone1Name=dyn_backStrapIn_02_l, bone2Name=dyn_backStrapIn_01_l, swing1LimitDegrees=18, swing2LimitDegrees=6, twistLimitDegrees=2}
 {
  bone1Name=dyn_backStrapIn_01_Base_r
  bone2Name=pelvis
  swing1LimitDegrees=45
  swing1Motion=Free
  swing2LimitDegrees=45
  swing2Motion=Free
  twistLimitDegrees=45
  twistMotion=Free
 }
 {bone1Name=dyn_backStrapIn_01_r, bone2Name=dyn_backStrapIn_01_Base_r, swing1LimitDegrees=18, swing2LimitDegrees=6, twistLimitDegrees=2}
 {bone1Name=dyn_backStrapIn_02_r, bone2Name=dyn_backStrapIn_01_r, swing1LimitDegrees=18, swing2LimitDegrees=6, twistLimitDegrees=2}
 {
  bone1Name=dyn_backStrapOut_01_Base_r
  bone2Name=pelvis
  swing1LimitDegrees=45
  swing1Motion=Free
  swing2LimitDegrees=45
  swing2Motion=Free
  twistLimitDegrees=45
  twistMotion=Free
 }
 {bone1Name=dyn_backStrapOut_01_r, bone2Name=dyn_backStrapOut_01_Base_r, swing1LimitDegrees=18, swing2LimitDegrees=6, twistLimitDegrees=2}
 {bone1Name=dyn_backStrapOut_02_r, bone2Name=dyn_backStrapOut_01_r, swing1LimitDegrees=18, swing2LimitDegrees=6, twistLimitDegrees=2}
 {
  bone1Name=dyn_bigBeltMain_Base
  bone2Name=pelvis
  swing1LimitDegrees=45
  swing1Motion=Free
  swing2LimitDegrees=45
  swing2Motion=Free
  twistLimitDegrees=45
  twistMotion=Free
 }
 {
  bone1Name=dyn_bigBeltMain
  bone2Name=dyn_bigBeltMain_Base
  swing1LimitDegrees=45
  swing1Motion=Free
  swing2LimitDegrees=45
  swing2Motion=Free
  twistLimitDegrees=45
  twistMotion=Free
 }
 {
  bone1Name=thigh_l
  bone2Name=pelvis
  swing1LimitDegrees=45
  swing1Motion=Free
  swing2LimitDegrees=45
  swing2Motion=Free
  twistLimitDegrees=45
  twistMotion=Free
 }
]
