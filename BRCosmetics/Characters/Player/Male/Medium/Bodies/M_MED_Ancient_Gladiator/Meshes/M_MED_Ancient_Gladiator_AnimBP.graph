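# Post-process style AnimBlueprint for the Ancient Gladiator body cosmetic, animating its straps, chains and loincloth atop the parent mesh's copied pose.
# A state-driven Control Rig, thigh-angle chain offsets, RigidBody physics, weighted strap constraints and a strap Control Rig build a cached pose.
# Separate crouching and standing strap/chain simulations branch from that cache and blend on crouch state.
# Events update state alphas and reset dynamics after emotes or crouching end.
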AnimBlueprint /BRCosmetics/Characters/Player/Male/Medium/Bodies/M_MED_Ancient_Gladiator/Meshes/M_MED_Ancient_Gladiator_AnimBP
target_skeleton=&/M_MED_Ancient_Gladiator_Skeleton
graphs [
 AnimGraph {nodes [
   Constraint!(
    ComponentPose
    Alpha=1
    ConstraintWeights_0=0.004
    ConstraintWeights_1:float=0.0146
    ConstraintWeights_2:float=0.1737
    ConstraintWeights_3:float=0.039
    ConstraintWeights_4:float=0.0213
    ConstraintWeights_5:float=0.0004
    ConstraintWeights_6:float=0.5208
    ConstraintWeights_7:float=0.2261
   ) -> (Pose) {
    alphaBoolBlend {blendOption=Linear}
    constraintSetup [
     {targetBone {boneName=dyn_bigBeltMain}}
     {targetBone {boneName=thigh_twist_01_r}}
     {targetBone {boneName=spine_01}}
     {targetBone {boneName=spine_02}}
     {targetBone {boneName=spine_03}}
     {targetBone {boneName=spine_04}}
     {targetBone {boneName=pelvis}}
     {targetBone {boneName=thigh_twist_01_l}}
    ]
    constraintWeights [1,1,1,1,1,1,1,1]
    boneToModify=~
   }
   n1 = "Constraint - dyn_frontStrapOut_03_r (PPPPPPPPPPPPP)"!(
    ComponentPose
    Alpha=1
    ConstraintWeights_0=0.0763
    ConstraintWeights_1:float=0.306
    ConstraintWeights_2:float=0.0145
    ConstraintWeights_3:float=0.0896
    ConstraintWeights_4:float=0.2208
    ConstraintWeights_5:float=0.0087
    ConstraintWeights_6:float=0.0037
    ConstraintWeights_7:float=0.1785
    ConstraintWeights_8:float=0.0222
    ConstraintWeights_9:float=0.0176
    ConstraintWeights_10:float=0.0041
    ConstraintWeights_11:float=0.0553
    ConstraintWeights_12:float=0.0027
   ) -> (Pose) {
    alphaBoolBlend {blendOption=Linear}
    boneToModify {boneName=dyn_frontStrapOut_03_r}
    constraintSetup [
     {targetBone {boneName=dyn_loinClothFront_01_r}}
     {targetBone {boneName=dyn_loinClothFront_02_r}}
     {targetBone {boneName=dyn_loinClothFront_04_c}}
     {targetBone {boneName=dyn_loinClothFront_03_c}}
     {targetBone {boneName=thigh_twist_01_r}}
     {targetBone {boneName=dyn_loinClothFront_01_l}}
     {targetBone {boneName=dyn_loinClothFront_03_l}}
     {targetBone {boneName=dyn_loinClothFront_02_c}}
     {targetBone {boneName=pelvis}}
     {targetBone {boneName=dyn_loinClothFront_02_l}}
     {targetBone {boneName=thigh_twist_01_l}}
     {targetBone {boneName=dyn_loinClothFront_03_r}}
     {targetBone {boneName=dyn_loinClothFront_05_c}}
    ]
    constraintWeights [1,1,1,1,1,1,1,1,1,1,1,1,1]
   }
   n2 = "Constraint - dyn_frontStrapIn_03_r (PPPPPPPPPPPPP)"!(
    ComponentPose=@n1
    Alpha=1
    ConstraintWeights_0=0.098
    ConstraintWeights_1:float=0.1878
    ConstraintWeights_2:float=0.03
    ConstraintWeights_3:float=0.2324
    ConstraintWeights_4:float=0.0197
    ConstraintWeights_5:float=0.0421
    ConstraintWeights_6:float=0.0139
    ConstraintWeights_7:float=0.2383
    ConstraintWeights_8:float=0.0233
    ConstraintWeights_9:float=0.067
    ConstraintWeights_10:float=0.0106
    ConstraintWeights_11:float=0.0314
    ConstraintWeights_12:float=0.0055
   ) -> (Pose) {
    alphaBoolBlend {blendOption=Linear}
    boneToModify {boneName=dyn_frontStrapIn_03_r}
    constraintSetup [
     {targetBone {boneName=dyn_loinClothFront_01_r}}
     {targetBone {boneName=dyn_loinClothFront_02_r}}
     {targetBone {boneName=dyn_loinClothFront_04_c}}
     {targetBone {boneName=dyn_loinClothFront_03_c}}
     {targetBone {boneName=thigh_twist_01_r}}
     {targetBone {boneName=dyn_loinClothFront_01_l}}
     {targetBone {boneName=dyn_loinClothFront_03_l}}
     {targetBone {boneName=dyn_loinClothFront_02_c}}
     {targetBone {boneName=pelvis}}
     {targetBone {boneName=dyn_loinClothFront_02_l}}
     {targetBone {boneName=thigh_twist_01_l}}
     {targetBone {boneName=dyn_loinClothFront_03_r}}
     {targetBone {boneName=dyn_loinClothFront_05_c}}
    ]
    constraintWeights [1,1,1,1,1,1,1,1,1,1,1,1,1]
   }
   n3 = "Constraint - dyn_frontStrapIn_03_l (PPPPPPPPPPPPP)"!(
    ComponentPose=@n2
    Alpha=1
    ConstraintWeights_0=0.0398
    ConstraintWeights_1:float=0.0693
    ConstraintWeights_2:float=0.0291
    ConstraintWeights_3:float=0.2381
    ConstraintWeights_4:float=0.0116
    ConstraintWeights_5:float=0.1096
    ConstraintWeights_6:float=0.0398
    ConstraintWeights_7:float=0.2343
    ConstraintWeights_8:float=0.022
    ConstraintWeights_9:float=0.1739
    ConstraintWeights_10:float=0.0163
    ConstraintWeights_11:float=0.0111
    ConstraintWeights_12:float=0.0051
   ) -> (Pose) {
    alphaBoolBlend {blendOption=Linear}
    boneToModify {boneName=dyn_frontStrapIn_03_l}
    constraintSetup [
     {targetBone {boneName=dyn_loinClothFront_01_r}}
     {targetBone {boneName=dyn_loinClothFront_02_r}}
     {targetBone {boneName=dyn_loinClothFront_04_c}}
     {targetBone {boneName=dyn_loinClothFront_03_c}}
     {targetBone {boneName=thigh_twist_01_r}}
     {targetBone {boneName=dyn_loinClothFront_01_l}}
     {targetBone {boneName=dyn_loinClothFront_03_l}}
     {targetBone {boneName=dyn_loinClothFront_02_c}}
     {targetBone {boneName=pelvis}}
     {targetBone {boneName=dyn_loinClothFront_02_l}}
     {targetBone {boneName=thigh_twist_01_l}}
     {targetBone {boneName=dyn_loinClothFront_03_r}}
     {targetBone {boneName=dyn_loinClothFront_05_c}}
    ]
    constraintWeights [1,1,1,1,1,1,1,1,1,1,1,1,1]
   }
   "Constraint - dyn_frontStrapOut_03_l (PPPPPPPPPPPPP)"!(
    ComponentPose=@n3
    Alpha=1
    ConstraintWeights_0=0.0093
    ConstraintWeights_1:float=0.0192
    ConstraintWeights_2:float=0.0124
    ConstraintWeights_3:float=0.0989
    ConstraintWeights_4:float=0.0047
    ConstraintWeights_5:float=0.1043
    ConstraintWeights_6:float=0.0583
    ConstraintWeights_7:float=0.115
    ConstraintWeights_8:float=0.0212
    ConstraintWeights_9:float=0.2607
    ConstraintWeights_10:float=0.2899
    ConstraintWeights_11:float=0.0033
    ConstraintWeights_12:float=0.0028
   ) -> (Pose) {
    alphaBoolBlend {blendOption=Linear}
    boneToModify {boneName=dyn_frontStrapOut_03_l}
    constraintSetup [
     {targetBone {boneName=dyn_loinClothFront_01_r}}
     {targetBone {boneName=dyn_loinClothFront_02_r}}
     {targetBone {boneName=dyn_loinClothFront_04_c}}
     {targetBone {boneName=dyn_loinClothFront_03_c}}
     {targetBone {boneName=thigh_twist_01_r}}
     {targetBone {boneName=dyn_loinClothFront_01_l}}
     {targetBone {boneName=dyn_loinClothFront_03_l}}
     {targetBone {boneName=dyn_loinClothFront_02_c}}
     {targetBone {boneName=pelvis}}
     {targetBone {boneName=dyn_loinClothFront_02_l}}
     {targetBone {boneName=thigh_twist_01_l}}
     {targetBone {boneName=dyn_loinClothFront_03_r}}
     {targetBone {boneName=dyn_loinClothFront_05_c}}
    ]
    constraintWeights [1,1,1,1,1,1,1,1,1,1,1,1,1]
   }
   Local To Component()
   n4 = "Transform (Modify) Bone - Bone: dyn_chainFront_01_l"(Alpha=Get thigh_rot_l(), Translation="0.021,3.442,-4.673", Rotation="6.390,-26.684,36.503", Scale="1,1,1") {
    rotationMode=BMM_Additive
    rotationSpace=BCS_BoneSpace
    translationMode=BMM_Additive
    translationSpace=BCS_BoneSpace
    alphaBoolBlend {blendOption=Linear}
    boneToModify {boneName=dyn_chainFront_01_l}
    rotation=<rot 6.39 -26.684 36.5031>
    translation=<0.021 3.442 -4.673>
   }
   "Transform (Modify) Bone - Bone: dyn_chainFront_01_r"(ComponentPose=@n4, Alpha=Get thigh_rot_r(), Translation="0.324,-1.843,-0.113", Rotation="45,-65,11.298", Scale="1,1,1") {
    rotationMode=BMM_Additive
    rotationSpace=BCS_BoneSpace
    translationMode=BMM_Additive
    translationSpace=BCS_BoneSpace
    alphaBoolBlend {blendOption=Linear}
    boneToModify {boneName=dyn_chainFront_01_r}
    rotation=<rot 45 -65 11.298>
    translation=<0.324 -1.843 -0.113>
   }
   RigidBody() {
    bClampLinearTranslationLimitToRefPose=true
    bOverrideWorldGravity=true
    overridePhysicsAsset=&/M_MED_Ancient_Gladiator_Chain_Physics
    simulationSpace=BaseBoneSpace
    alphaBoolBlend {blendOption=Linear}
    baseBoneRef {boneName=pelvis}
    componentAppliedLinearAccClamp=<2000 2000 2000>
    componentLinearAccScale=<0.2 0.2 0.2>
    componentLinearVelScale=<0.3 0.3 0.3>
    overrideWorldGravity {z=-980}
    simSpaceSettings {maxAngularAcceleration=0, maxAngularVelocity=0, maxLinearAcceleration=800, maxLinearVelocity=300, velocityScaleZ=0, worldAlpha=1}
   }
   n5 = RigidBody(ComponentPose=Local To Component(LocalPose=Use cached pose 'preStrap'())) {
    bClampLinearTranslationLimitToRefPose=true
    bOverrideWorldGravity=true
    overridePhysicsAsset=&/M_MED_AncientGladiator_StrapCRPhysics
    simulationSpace=BaseBoneSpace
    alphaBoolBlend {blendOption=Linear}
    baseBoneRef {boneName=pelvis}
    componentAppliedLinearAccClamp=<2000 2000 2000>
    componentLinearAccScale=<0.5 0.5 0.5>
    componentLinearVelScale=<0.1 0.1 0.1>
    overrideWorldGravity {z=-490}
   }
   n6 = RigidBody(ComponentPose=@n5) {
    bClampLinearTranslationLimitToRefPose=true
    bOverrideWorldGravity=true
    overridePhysicsAsset=&/M_MED_Ancient_Gladiator_Chain_Crouch_Physics
    simulationSpace=BaseBoneSpace
    alphaBoolBlend {blendOption=Linear}
    baseBoneRef {boneName=pelvis}
    componentAppliedLinearAccClamp=<2000 2000 2000>
    componentLinearAccScale=<0.2 0.2 0.2>
    componentLinearVelScale=<0.3 0.3 0.3>
    overrideWorldGravity {z=-980}
    simSpaceSettings {maxAngularAcceleration=0, maxAngularVelocity=0, maxLinearAcceleration=800, maxLinearVelocity=300, velocityScaleZ=0, worldAlpha=1}
   }
   n7 = RigidBody(ComponentPose=Local To Component(LocalPose=Use cached pose 'preStrap'())) {
    bClampLinearTranslationLimitToRefPose=true
    bOverrideWorldGravity=true
    overridePhysicsAsset=&/M_MED_AncientGladiator_StrapPhysics
    simulationSpace=BaseBoneSpace
    alphaBoolBlend {blendOption=Linear}
    baseBoneRef {boneName=pelvis}
    componentAppliedLinearAccClamp=<2000 2000 2000>
    componentLinearAccScale=<0.3 0.3 0.3>
    componentLinearVelScale=<0.1 0.1 0.1>
    overrideWorldGravity {z=-490}
   }
   n8 = RigidBody(ComponentPose=@n7) {
    bClampLinearTranslationLimitToRefPose=true
    bOverrideWorldGravity=true
    overridePhysicsAsset=&M_MED_Ancient_Gladiator_Chain_Physics
    simulationSpace=BaseBoneSpace
    alphaBoolBlend {blendOption=Linear}
    baseBoneRef {boneName=pelvis}
    componentAppliedLinearAccClamp=<2000 2000 2000>
    componentLinearAccScale=<0.2 0.2 0.2>
    componentLinearVelScale=<0.3 0.3 0.3>
    overrideWorldGravity {z=-980}
    simSpaceSettings {maxAngularAcceleration=0, maxAngularVelocity=0, maxLinearAcceleration=800, maxLinearVelocity=300, velocityScaleZ=0, worldAlpha=1}
   }
   Output Pose(Result=Blend Poses by bool(
     BlendPose_0=Component To Local(ComponentPose=@n6)
     BlendPose_1=Component To Local(ComponentPose=@n8)
     BlendTime_0=0.15
     BlendTime_1=0.15
     bActiveValue=Get bIsCrouching()
    ) {blendType=Linear}) {layerGroup=None}
   n9 = Get isCrouching() -> (isCrouching:double)
   n10 = Control Rig!(
    Source=Copy Pose From Mesh() {bUseAttachedParent=true}
    Alpha
    targeting:double=Get isAiming() -> (isAiming:double)
    crouching:double=@n9
    skydiving:double=Get isSkydiving() -> (isSkydiving:double)
    hanggliding:double=Get isGliding() -> (isGliding:double)
   ) -> (Pose) {
    alphaBoolBlend {blendOption=Linear}
    controlRigAssetReference {blueprintRigClass=&/M_MED_AncientGladiator_Body_CtrlRig.M_MED_AncientGladiator_Body_CtrlRig_C}
   }
   n11 = "Transform (Modify) Bone - Bone: dyn_chainFront_01_l"(
    ComponentPose=Local To Component(LocalPose=@n10)
    Alpha=Get thigh_rot_l()
    Translation="0.021,3.442,-4.673"
    Rotation="6.390,-26.684,36.503"
    Scale="1,1,1"
   ) {
    rotationMode=BMM_Additive
    rotationSpace=BCS_BoneSpace
    translationMode=BMM_Additive
    translationSpace=BCS_BoneSpace
    alphaBoolBlend {blendOption=Linear}
    boneToModify {boneName=dyn_chainFront_01_l}
    rotation=<rot 6.39 -26.684 36.5031>
    translation=<0.021 3.442 -4.673>
   }
   n12 = "Transform (Modify) Bone - Bone: dyn_chainFront_01_r"(ComponentPose=@n11, Alpha=Get thigh_rot_r(), Translation="0.416,-2.096,0.156", Rotation="29.771,-73.665,-5.251", Scale="1,1,1") {
    rotationMode=BMM_Additive
    rotationSpace=BCS_BoneSpace
    translationMode=BMM_Additive
    translationSpace=BCS_BoneSpace
    alphaBoolBlend {blendOption=Linear}
    boneToModify {boneName=dyn_chainFront_01_r}
    rotation=<rot 29.771 -73.665 -5.251>
    translation=<0.416 -2.096 0.156>
   }
   n13 = RigidBody(ComponentPose=@n12) {
    bClampLinearTranslationLimitToRefPose=true
    bOverrideWorldGravity=true
    overridePhysicsAsset=&/M_MED_Ancient_Gladiator_Physics
    simulationSpace=BaseBoneSpace
    alphaBoolBlend {blendOption=Linear}
    baseBoneRef {boneName=pelvis}
    componentAppliedLinearAccClamp=<2000 2000 2000>
    componentLinearAccScale=<0.2 0.2 0.2>
    componentLinearVelScale=<0.3 0.3 0.3>
    overrideWorldGravity {z=-980}
    simSpaceSettings {maxAngularAcceleration=0, maxAngularVelocity=0, maxLinearAcceleration=800, maxLinearVelocity=300, velocityScaleZ=0, worldAlpha=1}
   }
   n14 = "Constraint - dyn_frontStrapOut_01_Base_r (PPPPPPPPPP)"!(
    ComponentPose=@n13
    Alpha=1
    ConstraintWeights_0=0.0792
    ConstraintWeights_1:float=0.0462
    ConstraintWeights_2:float=0.0067
    ConstraintWeights_3:float=0.015
    ConstraintWeights_4:float=0.4308
    ConstraintWeights_5:float=0.0366
    ConstraintWeights_6:float=0.0393
    ConstraintWeights_7:float=0.1175
    ConstraintWeights_8:float=0.224
    ConstraintWeights_9:float=0.0048
   ) -> (Pose) {
    alphaBoolBlend {blendOption=Linear}
    boneToModify {boneName=dyn_frontStrapOut_01_Base_r}
    constraintSetup [
     {targetBone {boneName=dyn_loinClothFront_01_r}}
     {targetBone {boneName=dyn_loinClothFront_02_r}}
     {targetBone {boneName=thigh_r}}
     {targetBone {boneName=dyn_loinClothFront_03_c}}
     {targetBone {boneName=thigh_twist_01_r}}
     {targetBone {boneName=spine_01}}
     {targetBone {boneName=thigh_twist_01_l}}
     {targetBone {boneName=dyn_loinClothFront_02_c}}
     {targetBone {boneName=pelvis}}
     {targetBone {boneName=dyn_loinClothFront_01_l}}
    ]
    constraintWeights [1,1,1,1,1,1,1,1,1,1]
   }
   n15 = "Constraint - dyn_frontStrapIn_01_Base_r (PPPPPPPPPPPP)"!(
    ComponentPose=@n14
    Alpha=1
    ConstraintWeights_0=0.0665
    ConstraintWeights_1:float=0.0249
    ConstraintWeights_2:float=0.0054
    ConstraintWeights_3:float=0.0324
    ConstraintWeights_4:float=0.1888
    ConstraintWeights_5:float=0.038
    ConstraintWeights_6:float=0.0052
    ConstraintWeights_7:float=0.0914
    ConstraintWeights_8:float=0.2598
    ConstraintWeights_9:float=0.2534
    ConstraintWeights_10:float=0.0087
    ConstraintWeights_11:float=0.0255
   ) -> (Pose) {
    alphaBoolBlend {blendOption=Linear}
    boneToModify {boneName=dyn_frontStrapIn_01_Base_r}
    constraintSetup [
     {targetBone {boneName=dyn_loinClothFront_01_r}}
     {targetBone {boneName=dyn_loinClothFront_02_r}}
     {targetBone {boneName=thigh_r}}
     {targetBone {boneName=dyn_loinClothFront_03_c}}
     {targetBone {boneName=thigh_twist_01_r}}
     {targetBone {boneName=spine_01}}
     {targetBone {boneName=spine_02}}
     {targetBone {boneName=thigh_twist_01_l}}
     {targetBone {boneName=dyn_loinClothFront_02_c}}
     {targetBone {boneName=pelvis}}
     {targetBone {boneName=dyn_loinClothFront_02_l}}
     {targetBone {boneName=dyn_loinClothFront_01_l}}
    ]
    constraintWeights [1,1,1,1,1,1,1,1,1,1,1,1]
   }
   n16 = "Constraint - dyn_frontStrapIn_01_Base_l (PPPPPPPPPPPP)"!(
    ComponentPose=@n15
    Alpha=1
    ConstraintWeights_0=0.0248
    ConstraintWeights_1:float=0.0088
    ConstraintWeights_2:float=0.0339
    ConstraintWeights_3:float=0.0958
    ConstraintWeights_4:float=0.0381
    ConstraintWeights_5:float=0.0053
    ConstraintWeights_6:float=0.1579
    ConstraintWeights_7:float=0.2642
    ConstraintWeights_8:float=0.0052
    ConstraintWeights_9:float=0.2681
    ConstraintWeights_10:float=0.0259
    ConstraintWeights_11:float=0.072
   ) -> (Pose) {
    alphaBoolBlend {blendOption=Linear}
    boneToModify {boneName=dyn_frontStrapIn_01_Base_l}
    constraintSetup [
     {targetBone {boneName=dyn_loinClothFront_01_r}}
     {targetBone {boneName=dyn_loinClothFront_02_r}}
     {targetBone {boneName=dyn_loinClothFront_03_c}}
     {targetBone {boneName=thigh_twist_01_r}}
     {targetBone {boneName=spine_01}}
     {targetBone {boneName=spine_02}}
     {targetBone {boneName=thigh_twist_01_l}}
     {targetBone {boneName=dyn_loinClothFront_02_c}}
     {targetBone {boneName=thigh_l}}
     {targetBone {boneName=pelvis}}
     {targetBone {boneName=dyn_loinClothFront_02_l}}
     {targetBone {boneName=dyn_loinClothFront_01_l}}
    ]
    constraintWeights [1,1,1,1,1,1,1,1,1,1,1,1]
   }
   n17 = "Constraint - dyn_frontStrapOut_01_Base_l (PPPPPPPPP)"!(
    ComponentPose=@n16
    Alpha=1
    ConstraintWeights_0=0.013
    ConstraintWeights_1:float=0.0285
    ConstraintWeights_2:float=0.0239
    ConstraintWeights_3:float=0.4155
    ConstraintWeights_4:float=0.125
    ConstraintWeights_5:float=0.0028
    ConstraintWeights_6:float=0.2487
    ConstraintWeights_7:float=0.0233
    ConstraintWeights_8:float=0.1192
   ) -> (Pose) {
    alphaBoolBlend {blendOption=Linear}
    boneToModify {boneName=dyn_frontStrapOut_01_Base_l}
    constraintSetup [
     {targetBone {boneName=dyn_loinClothFront_03_c}}
     {targetBone {boneName=thigh_twist_01_r}}
     {targetBone {boneName=spine_01}}
     {targetBone {boneName=thigh_twist_01_l}}
     {targetBone {boneName=dyn_loinClothFront_02_c}}
     {targetBone {boneName=thigh_l}}
     {targetBone {boneName=pelvis}}
     {targetBone {boneName=dyn_loinClothFront_02_l}}
     {targetBone {boneName=dyn_loinClothFront_01_l}}
    ]
    constraintWeights [1,1,1,1,1,1,1,1,1]
   }
   n18 = "Constraint - dyn_frontStrapOut_02_r (PPPPPPPPPPP)"!(
    ComponentPose=@n17
    Alpha=1
    ConstraintWeights_0=0.0917
    ConstraintWeights_1:float=0.1516
    ConstraintWeights_2:float=0.0457
    ConstraintWeights_3:float=0.3768
    ConstraintWeights_4:float=0.0088
    ConstraintWeights_5:float=0.0088
    ConstraintWeights_6:float=0.018
    ConstraintWeights_7:float=0.2073
    ConstraintWeights_8:float=0.072
    ConstraintWeights_9:float=0.0062
    ConstraintWeights_10:float=0.013
   ) -> (Pose) {
    alphaBoolBlend {blendOption=Linear}
    boneToModify {boneName=dyn_frontStrapOut_02_r}
    constraintSetup [
     {targetBone {boneName=dyn_loinClothFront_01_r}}
     {targetBone {boneName=dyn_loinClothFront_02_r}}
     {targetBone {boneName=dyn_loinClothFront_03_c}}
     {targetBone {boneName=thigh_twist_01_r}}
     {targetBone {boneName=dyn_loinClothFront_01_l}}
     {targetBone {boneName=spine_01}}
     {targetBone {boneName=thigh_twist_01_l}}
     {targetBone {boneName=dyn_loinClothFront_02_c}}
     {targetBone {boneName=pelvis}}
     {targetBone {boneName=dyn_loinClothFront_02_l}}
     {targetBone {boneName=dyn_loinClothFront_03_r}}
    ]
    constraintWeights [1,1,1,1,1,1,1,1,1,1,1]
   }
   n19 = "Constraint - dyn_frontStrapIn_02_r (PPPPPPPPPPPPP)"!(
    ComponentPose=@n18
    Alpha=1
    ConstraintWeights_0=0.1171
    ConstraintWeights_1:float=0.1379
    ConstraintWeights_2:float=0.0131
    ConstraintWeights_3:float=0.1734
    ConstraintWeights_4:float=0.0425
    ConstraintWeights_5:float=0.0486
    ConstraintWeights_6:float=0.008
    ConstraintWeights_7:float=0.0057
    ConstraintWeights_8:float=0.3153
    ConstraintWeights_9:float=0.0531
    ConstraintWeights_10:float=0.0493
    ConstraintWeights_11:float=0.0232
    ConstraintWeights_12:float=0.0129
   ) -> (Pose) {
    alphaBoolBlend {blendOption=Linear}
    boneToModify {boneName=dyn_frontStrapIn_02_r}
    constraintSetup [
     {targetBone {boneName=dyn_loinClothFront_01_r}}
     {targetBone {boneName=dyn_loinClothFront_02_r}}
     {targetBone {boneName=dyn_loinClothFront_04_c}}
     {targetBone {boneName=dyn_loinClothFront_03_c}}
     {targetBone {boneName=thigh_twist_01_r}}
     {targetBone {boneName=dyn_loinClothFront_01_l}}
     {targetBone {boneName=spine_01}}
     {targetBone {boneName=dyn_loinClothFront_03_l}}
     {targetBone {boneName=dyn_loinClothFront_02_c}}
     {targetBone {boneName=pelvis}}
     {targetBone {boneName=dyn_loinClothFront_02_l}}
     {targetBone {boneName=thigh_twist_01_l}}
     {targetBone {boneName=dyn_loinClothFront_03_r}}
    ]
    constraintWeights [1,1,1,1,1,1,1,1,1,1,1,1,1]
   }
   n20 = "Constraint - dyn_frontStrapIn_02_l (PPPPPPPPPPPP)"!(
    ComponentPose=@n19
    Alpha=1
    ConstraintWeights_0=0.0448
    ConstraintWeights_1:float=0.049
    ConstraintWeights_2:float=0.0125
    ConstraintWeights_3:float=0.1741
    ConstraintWeights_4:float=0.0238
    ConstraintWeights_5:float=0.0073
    ConstraintWeights_6:float=0.0163
    ConstraintWeights_7:float=0.3
    ConstraintWeights_8:float=0.048
    ConstraintWeights_9:float=0.1285
    ConstraintWeights_10:float=0.0713
    ConstraintWeights_11:float=0.1245
   ) -> (Pose) {
    alphaBoolBlend {blendOption=Linear}
    boneToModify {boneName=dyn_frontStrapIn_02_l}
    constraintSetup [
     {targetBone {boneName=dyn_loinClothFront_01_r}}
     {targetBone {boneName=dyn_loinClothFront_02_r}}
     {targetBone {boneName=dyn_loinClothFront_04_c}}
     {targetBone {boneName=dyn_loinClothFront_03_c}}
     {targetBone {boneName=thigh_twist_01_r}}
     {targetBone {boneName=spine_01}}
     {targetBone {boneName=dyn_loinClothFront_03_l}}
     {targetBone {boneName=dyn_loinClothFront_02_c}}
     {targetBone {boneName=pelvis}}
     {targetBone {boneName=dyn_loinClothFront_02_l}}
     {targetBone {boneName=thigh_twist_01_l}}
     {targetBone {boneName=dyn_loinClothFront_01_l}}
    ]
    constraintWeights [1,1,1,1,1,1,1,1,1,1,1,1]
   }
   n21 = "Constraint - dyn_frontStrapOut_02_l (PPPPPPPPPPP)"!(
    ComponentPose=@n20
    Alpha=1
    ConstraintWeights_0=0.012
    ConstraintWeights_1:float=0.0074
    ConstraintWeights_2:float=0.0447
    ConstraintWeights_3:float=0.0146
    ConstraintWeights_4:float=0.0086
    ConstraintWeights_5:float=0.0062
    ConstraintWeights_6:float=0.1308
    ConstraintWeights_7:float=0.1239
    ConstraintWeights_8:float=0.1308
    ConstraintWeights_9:float=0.4184
    ConstraintWeights_10:float=0.1027
   ) -> (Pose) {
    alphaBoolBlend {blendOption=Linear}
    boneToModify {boneName=dyn_frontStrapOut_02_l}
    constraintSetup [
     {targetBone {boneName=dyn_loinClothFront_03_l}}
     {targetBone {boneName=dyn_loinClothFront_02_r}}
     {targetBone {boneName=dyn_loinClothFront_03_c}}
     {targetBone {boneName=thigh_twist_01_r}}
     {targetBone {boneName=spine_01}}
     {targetBone {boneName=dyn_loinClothFront_01_r}}
     {targetBone {boneName=dyn_loinClothFront_02_c}}
     {targetBone {boneName=pelvis}}
     {targetBone {boneName=dyn_loinClothFront_02_l}}
     {targetBone {boneName=thigh_twist_01_l}}
     {targetBone {boneName=dyn_loinClothFront_01_l}}
    ]
    constraintWeights [1,1,1,1,1,1,1,1,1,1,1]
   }
   n22 = "Transform (Modify) Bone - Bone: dyn_frontStrapOut_02_l"(ComponentPose=@n21, Alpha=1, Translation="0.043,-1.303,-0.680", Rotation="-4.975,-0.066,0.494", Scale="1,1,1") {
    rotationMode=BMM_Additive
    rotationSpace=BCS_BoneSpace
    translationMode=BMM_Additive
    translationSpace=BCS_BoneSpace
    alphaBoolBlend {blendOption=Linear}
    boneToModify {boneName=dyn_frontStrapOut_02_l}
    rotation=<rot -4.975 -0.066 0.494>
    translation=<0.043 -1.303 -0.68>
   }
   n23 = "Transform (Modify) Bone - Bone: dyn_frontStrapOut_01_r"(ComponentPose=@n22, Alpha=1, Translation="0.042,-1.338,0.000", Rotation="4.984,-6.990,-0.400", Scale="1,1,1") {
    rotationMode=BMM_Additive
    rotationSpace=BCS_BoneSpace
    translationMode=BMM_Additive
    translationSpace=BCS_BoneSpace
    alphaBoolBlend {blendOption=Linear}
    boneToModify {boneName=dyn_frontStrapOut_01_r}
    rotation=<rot 4.984 -6.9904 -0.4>
    translation=<0.042 -1.338>
   }
   n24 = "Transform (Modify) Bone - Bone: dyn_frontStrapIn_01_l"(ComponentPose=@n23, Alpha=1, Translation="0.000,-1.080,0.000", Rotation="0.000,-5.000,0.000", Scale="1,1,1") {
    rotationMode=BMM_Additive
    rotationSpace=BCS_BoneSpace
    translationMode=BMM_Additive
    translationSpace=BCS_BoneSpace
    alphaBoolBlend {blendOption=Linear}
    boneToModify {boneName=dyn_frontStrapIn_01_l}
    rotation {yaw=-5.0001}
    translation {y=-1.08}
   }
   n25 = "Transform (Modify) Bone - Bone: dyn_frontStrapIn_01_r"(ComponentPose=@n24, Alpha=1, Translation="0.000,-1.080,0.000", Rotation="0.000,-5.000,0.000", Scale="1,1,1") {
    rotationMode=BMM_Additive
    rotationSpace=BCS_BoneSpace
    translationMode=BMM_Additive
    translationSpace=BCS_BoneSpace
    alphaBoolBlend {blendOption=Linear}
    boneToModify {boneName=dyn_frontStrapIn_01_r}
    rotation {yaw=-5.0001}
    translation {y=-1.08}
   }
   n26 = Control Rig!(Source=Component To Local(ComponentPose=@n25), Alpha, crouching=@n9) -> (Pose) {
    alphaBoolBlend {blendOption=Linear}
    controlRigAssetReference {blueprintRigClass=&/M_MED_AncientGladiator_StrapsCtrlRig.M_MED_AncientGladiator_StrapsCtrlRig_C}
   }
   Save cached pose 'preStrap'(Pose=@n26) {cacheName=preStrap}
  ]}
 EventGraph {nodes [
   Event BlueprintUpdateAnimation()
   > Set isSprinting(isSprinting:double="To Float (Boolean)"(InBool=Get bIsSprinting())) -> (Output_Get:double)
   > Set isGliding(isGliding:double="To Float (Boolean)"(InBool=Get bIsParachuteOpened())) -> (Output_Get:double)
   n1 = Get Skel_Mesh()
   > Set thigh_rot_l(thigh_rot_l=GetAlphaFromReferenceTransformDelta(SkeletalMeshComponent=@n1, BoneName=thigh_l, BaseName=pelvis, InputRange=(X=-20,Y=-115), TwistAxis="1,0,0"))
   > Set thigh_rot_r(thigh_rot_r=GetAlphaFromReferenceTransformDelta(SkeletalMeshComponent=@n1, BoneName=thigh_r, BaseName=pelvis, InputRange=(X=-20,Y=-100), TwistAxis="1,0,0"))
   Event OnBeginSkydiving()
   n2 = "To Float (Boolean)"(InBool=Get bIsSkyDiving())
   > Set isSkydiving(isSkydiving:double=@n2) -> (Output_Get:double)
   > Set isSkydivingINV(isSkydivingINV:double=float - float(B=@n2) {is=title K2Node_CallFunction}) -> (Output_Get:double)
   n3 = Event OnEndPlayingEmote() {is=title K2Node_Event}
   Delay(execute=[@n3.then, @n5.then], Duration=0.1)
   > ResetDynamics(InTeleportType=TeleportPhysics)
   > ResetDynamics(InTeleportType=ResetPhysics)
   Event BlueprintInitializeAnimation()
   > Set Skel_Mesh(Skel_Mesh=GetOwnerSkeletalMeshComponent())
   Event OnEndTargeting() {is=title K2Node_Event}
   > Set isAiming(isAiming:double="To Float (Boolean)"(InBool=Get bIsTargeting())) -> (Output_Get:double)
   Event OnEndSkydiving()
   n4 = "To Float (Boolean)"(InBool=Get bIsSkyDiving())
   > Set isSkydiving(isSkydiving=@n4)
   > Set isSkydivingINV(isSkydivingINV=float - float(B=@n4))
   Event OnEndCrouching()
   > n5 = Set isCrouching(isCrouching:double="To Float (Boolean)"(InBool=Get bIsCrouching())) -> (Output_Get:double)
   Event OnBeginCrouching()
   > Set isCrouching(isCrouching="To Float (Boolean)"(InBool=Get bIsCrouching()))
   Event OnBeginTargeting() {is=title K2Node_Event}
   > Set isAiming(isAiming="To Float (Boolean)"(InBool=Get bIsTargeting()))
  ]}
]
variables [
 isAiming {type=number}
 isCrouching {type=number}
 isGliding {type=number}
 isSkydiving {type=number}
 isSkydivingINV {type=number}
 isSprinting {type=number}
 NewVar_0 {type=number}
 Skel_Mesh {type=SkeletalMeshComponent}
 thigh_rot_l {type=number}
 thigh_rot_r {type=number}
]
